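# Physics asset for the female medium RoosterRoast body cosmetic, setting up physical motion for its hanging apron.
# Three capsule chains (center, left, right) each run from a small base body through four dynamic bones, with narrow swing limits and mostly locked twist for stiff, hinge-like movement.
# Each base joins spine_01 with free angular motion.
# Larger pelvis, thigh and spine capsules provide body volumes around the apron.
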PhysicsAsset /BRCosmetics/Characters/Player/Female/Medium/Bodies/F_MED_RoosterRoast/Meshes/F_MED_RoosterRoast_Physics
bodies {
 base_apron_01_cr [{shapeType=Capsule, length=0.505, radius=0.505, shapeName=base_apron_01_cr_capsule, rotation=<rot 0 -90 0>}]
 base_apron_01_lf [{shapeType=Capsule, length=0.505, radius=0.505, shapeName=base_apron_01_lf_capsule, rotation=<rot 0 -90 0>}]
 base_apron_01_rt [{shapeType=Capsule, length=0.505, radius=0.505, shapeName=base_apron_01_rt_capsule, rotation=<rot 0 -90 0>}]
 dfrm_pelvis [{shapeType=Capsule, length=6.115, radius=16.033, center=<2.2526 -1.2454 1.0946>}]
 dfrm_thigh_l [{shapeType=Capsule, length=24.6094, radius=8.5386, center=<-16.5255 2.1404 -1.6287>, rotation=<rot 90 4.9996 0>}]
 dfrm_thigh_r [{shapeType=Capsule, length=24.6094, radius=8.5386, center=<17.2923 -2.6537 0.7883>, rotation=<rot 90 9.9998 0>}]
 dyn_apron_01_cr [{shapeType=Capsule, length=7, radius=2, shapeName=dyn_apron_01_cr_capsule, center=<4 0 0>, rotation=<rot 90 0 0>}]
 dyn_apron_01_lf [{shapeType=Capsule, length=7, radius=2, shapeName=dyn_apron_01_lf_capsule, center=<4 0 0>, rotation=<rot 90 0 0>}]
 dyn_apron_01_rt [{shapeType=Capsule, length=7, radius=2, shapeName=dyn_apron_01_rt_capsule, center=<4 0 0>, rotation=<rot 90 0 0>}]
 dyn_apron_02_cr [{shapeType=Capsule, length=6, radius=3, shapeName=dyn_apron_02_cr_capsule, center=<4 0 0>, rotation=<rot 90 0 0>}]
 dyn_apron_02_lf [{shapeType=Capsule, length=7, radius=2, shapeName=dyn_apron_02_lf_capsule, center=<4 0 0>, rotation=<rot 90 0 0>}]
 dyn_apron_02_rt [{shapeType=Capsule, length=7, radius=2, shapeName=dyn_apron_02_rt_capsule, center=<4 0 0>, rotation=<rot 90 0 0>}]
 dyn_apron_03_cr [{shapeType=Capsule, length=4.7619, radius=4.2381, shapeName=dyn_apron_03_cr_capsule, center=<4 0.859 0>, rotation=<rot 0 -0.0003 0>}]
 dyn_apron_03_lf [{shapeType=Capsule, length=6, radius=3, shapeName=dyn_apron_03_rt_capsule, center=<4 -1.7415 3.1489>, rotation=<rot 0 -0.0004 -30>}]
 dyn_apron_03_rt [{shapeType=Capsule, length=6.5, radius=2.5, shapeName=dyn_apron_03_rt_capsule, center=<5.0474 -1.0756 -2.6012>, rotation=<rot 0 -0.0002 25>}]
 dyn_apron_04_cr [{
   shapeType=Capsule
   length=11.2648
   radius=2.505
   shapeName=dyn_apron_04_cr_capsule
   center=<6.1494 1.6375 -0.0938>
   rotation=<rot 0 -6.1852 3.2587>
  }]
 dyn_apron_04_lf [{shapeType=Capsule, length=9.1563, radius=2.6806, shapeName=dyn_apron_03_rt_capsule, center=<4 -1.7415 3.1489>, rotation=<rot 0 -0.0004 -30>}]
 dyn_apron_04_rt [{
   shapeType=Capsule
   length=9.3885
   radius=2.4874
   shapeName=dyn_apron_03_rt_capsule
   center=<5.0474 -1.0756 -2.6012>
   rotation=<rot 0 -0.0002 25>
  }]
 spine_01 [{shapeType=Capsule, length=12, radius=12, shapeName=spine_01_capsule, center=<-4.7837 0 0>}]
}
constraints [
 {bone1Name=dfrm_thigh_r, bone2Name=dfrm_pelvis, swing1LimitDegrees=45, swing2LimitDegrees=45, twistLimitDegrees=45}
 {bone1Name=dfrm_thigh_l, bone2Name=dfrm_pelvis, swing1LimitDegrees=45, swing2LimitDegrees=45, twistLimitDegrees=45}
 {
  bone1Name=base_apron_01_cr
  bone2Name=spine_01
  swing1LimitDegrees=45
  swing1Motion=Free
  swing2LimitDegrees=45
  swing2Motion=Free
  twistLimitDegrees=45
  twistMotion=Free
 }
 {bone1Name=dyn_apron_01_cr, bone2Name=base_apron_01_cr, swing1LimitDegrees=10, swing2LimitDegrees=0, twistLimitDegrees=0}
 {bone1Name=dyn_apron_02_cr, bone2Name=dyn_apron_01_cr, swing1LimitDegrees=5, swing2LimitDegrees=0, twistLimitDegrees=0, twistMotion=Locked}
 {bone1Name=dyn_apron_03_cr, bone2Name=dyn_apron_02_cr, swing1LimitDegrees=15, swing2LimitDegrees=2, twistLimitDegrees=45, twistMotion=Locked}
 {
  bone1Name=base_apron_01_lf
  bone2Name=spine_01
  swing1LimitDegrees=45
  swing1Motion=Free
  swing2LimitDegrees=45
  swing2Motion=Free
  twistLimitDegrees=45
  twistMotion=Free
 }
 {bone1Name=dyn_apron_01_lf, bone2Name=base_apron_01_lf, swing1LimitDegrees=10, swing2LimitDegrees=0, twistLimitDegrees=0}
 {bone1Name=dyn_apron_02_lf, bone2Name=dyn_apron_01_lf, swing1LimitDegrees=5, swing2LimitDegrees=0, twistLimitDegrees=0, twistMotion=Locked}
 {bone1Name=dyn_apron_03_lf, bone2Name=dyn_apron_02_lf, swing1LimitDegrees=15, swing2LimitDegrees=5, twistLimitDegrees=45, twistMotion=Locked}
 {
  bone1Name=base_apron_01_rt
  bone2Name=spine_01
  swing1LimitDegrees=45
  swing1Motion=Free
  swing2LimitDegrees=45
  swing2Motion=Free
  twistLimitDegrees=45
  twistMotion=Free
 }
 {bone1Name=dyn_apron_01_rt, bone2Name=base_apron_01_rt, swing1LimitDegrees=10, swing2LimitDegrees=0, twistLimitDegrees=0}
 {bone1Name=dyn_apron_02_rt, bone2Name=dyn_apron_01_rt, swing1LimitDegrees=5, swing2LimitDegrees=0, twistLimitDegrees=0, twistMotion=Locked}
 {bone1Name=dyn_apron_03_rt, bone2Name=dyn_apron_02_rt, swing1LimitDegrees=15, swing2LimitDegrees=0, twistLimitDegrees=45, twistMotion=Locked}
 {bone1Name=dyn_apron_04_cr, bone2Name=dyn_apron_03_cr, swing1LimitDegrees=10, swing2LimitDegrees=5, twistLimitDegrees=45, twistMotion=Locked}
 {bone1Name=dyn_apron_04_lf, bone2Name=dyn_apron_03_lf, swing1LimitDegrees=10, swing2LimitDegrees=5, twistLimitDegrees=45, twistMotion=Locked}
 {bone1Name=dyn_apron_04_rt, bone2Name=dyn_apron_03_rt, swing1LimitDegrees=10, swing2LimitDegrees=0, twistLimitDegrees=45, twistMotion=Locked}
]
