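# Animation Blueprint for the female medium RoosterRoast cosmetic body, adding procedural and secondary motion to the pose copied from its attached parent mesh.
# A Control Rig takes crouch, sprint, emote, pistol and cape inputs.
# Chained additive bone transforms then offset tail, wings, apron, ties and calf, including a cape-floating adjustment.
# Cached poses feed spring-based Anim Dynamics on wing and calf feather bones, with AnimDynPose as the output.
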AnimBlueprint /BRCosmetics/Characters/Player/Female/Medium/Bodies/F_MED_RoosterRoast/Meshes/F_MED_RoosterRoast_AnimBP
target_skeleton=&/F_MED_RoosterRoast_Skeleton
graphs [
 AnimGraph {nodes [
   Output Pose(Result=Use cached pose 'AnimDynPose'()) {layerGroup=None}
   n1 = Control Rig!(
    Source=Copy Pose From Mesh() {bUseAttachedParent=true}
    Alpha
    crouch:double=Get crouch() -> (crouch:double)
    Sprint:double=Get sprint() -> (sprint:double)
    emote:double=Get emote() -> (emote:double)
    Pistol:double=Get pistol() -> (pistol:double)
    IsCrouchMelee:bool=Get IsCrouchMelee() -> (IsCrouchMelee:bool)
    IsCape:bool=Get Is Wearing Cape() -> (Is Wearing Cape:bool)
   ) -> (Pose) {
    alphaBoolBlend {blendOption=Linear}
    controlRigAssetReference {blueprintRigClass=&/F_MED_RoosterRoast_CtrlRig.F_MED_RoosterRoast_CtrlRig_C}
   }
   Save cached pose 'ControlRigPose'(Pose=@n1) {cacheName=ControlRigPose}
   n2 = "Transform (Modify) Bone - Bone: base_tail_cr"!(
    ComponentPose=Local To Component(LocalPose=Use cached pose 'ControlRigPose'())
    bAlphaBoolEnabled:bool=True
    Translation="0,0,0"
    Rotation="-0.000,-20.000,-0.000"
    Scale="1,1,1"
   ) -> (Pose) {
    alphaInputType=Bool
    rotationMode=BMM_Additive
    rotationSpace=BCS_BoneSpace
    translationMode=BMM_Additive
    translationSpace=BCS_BoneSpace
    alphaBoolBlend {blendInTime=0.5, blendOutTime=0.5}
    boneToModify {boneName=base_tail_cr}
    rotation=<rot 0 -20 0>
   }
   n3 = "Transform (Modify) Bone - Bone: base_tail_right_01_cr"!(ComponentPose=@n2, bAlphaBoolEnabled:bool=True, Translation="0,0,0", Rotation="-0.000,-20.000,-0.000", Scale="1,1,1") -> (Pose) {
    alphaInputType=Bool
    rotationMode=BMM_Additive
    rotationSpace=BCS_BoneSpace
    translationMode=BMM_Additive
    translationSpace=BCS_BoneSpace
    alphaBoolBlend {blendInTime=0.5, blendOutTime=0.5}
    boneToModify {boneName=base_tail_right_01_cr}
    rotation=<rot 0 -20 0>
   }
   n4 = "Transform (Modify) Bone - Bone: dyn_tail_right_01_03_cr"!(ComponentPose=@n3, bAlphaBoolEnabled:bool=True, Translation="0,0,0", Rotation="0,-30,0", Scale="1,1,1") -> (Pose) {
    alphaInputType=Bool
    rotationMode=BMM_Additive
    rotationSpace=BCS_BoneSpace
    translationMode=BMM_Additive
    translationSpace=BCS_BoneSpace
    alphaBoolBlend {blendInTime=0.5, blendOutTime=0.5}
    boneToModify {boneName=dyn_tail_right_01_03_cr}
    rotation {yaw=-30}
   }
   n5 = "Transform (Modify) Bone - Bone: base_tail_left_01_cr"!(ComponentPose=@n4, bAlphaBoolEnabled:bool=True, Translation="0,0,0", Rotation="0,-10,0", Scale="1,1,1") -> (Pose) {
    alphaInputType=Bool
    rotationMode=BMM_Additive
    rotationSpace=BCS_BoneSpace
    translationMode=BMM_Additive
    translationSpace=BCS_BoneSpace
    alphaBoolBlend {blendInTime=0.5, blendOutTime=0.5}
    boneToModify {boneName=base_tail_left_01_cr}
    rotation {yaw=-10}
   }
   n6 = "Transform (Modify) Bone - Bone: base_tail_left_02_cr"!(ComponentPose=@n5, bAlphaBoolEnabled:bool=True, Translation="0,0,0", Rotation="-0.000,-25.000,-0.000", Scale="1,1,1") -> (Pose) {
    alphaInputType=Bool
    rotationMode=BMM_Additive
    rotationSpace=BCS_BoneSpace
    translationMode=BMM_Additive
    translationSpace=BCS_BoneSpace
    alphaBoolBlend {blendInTime=0.5, blendOutTime=0.5}
    boneToModify {boneName=base_tail_left_02_cr}
    rotation=<rot -0.0001 -25 0>
   }
   n7 = "Transform (Modify) Bone - Bone: dyn_wing_upper_l_twist_MSTR"!(ComponentPose=@n6, bAlphaBoolEnabled:bool=True, Translation="0,0,0", Rotation="-0.010,0.039,15.000", Scale="1,1,1") -> (Pose) {
    alphaInputType=Bool
    rotationMode=BMM_Additive
    rotationSpace=BCS_BoneSpace
    alphaBoolBlend {blendInTime=0.1, blendOption=Linear, blendOutTime=0.1}
    boneToModify {boneName=dyn_wing_upper_l_twist_MSTR}
    rotation=<rot -0.01 0.039 15>
   }
   n8 = "Transform (Modify) Bone - Bone: dyn_wing_upper_r_twist_MSTR"!(ComponentPose=@n7, bAlphaBoolEnabled:bool=True, Translation="0,0,0", Rotation="0,0,15", Scale="1,1,1") -> (Pose) {
    alphaInputType=Bool
    rotationMode=BMM_Additive
    rotationSpace=BCS_BoneSpace
    alphaBoolBlend {blendInTime=0.1, blendOption=Linear, blendOutTime=0.1}
    boneToModify {boneName=dyn_wing_upper_r_twist_MSTR}
    rotation {roll=15}
   }
   n9 = "Transform (Modify) Bone - Bone: dyn_wing_upper_l_twist_MSTR"!(ComponentPose=@n8, Alpha:float=1, Translation="0,0,0", Rotation="5,0,-5", Scale="1,1,1") -> (Pose) {
    rotationMode=BMM_Additive
    rotationSpace=BCS_BoneSpace
    alphaBoolBlend {blendInTime=0.1, blendOption=Linear, blendOutTime=0.1}
    boneToModify {boneName=dyn_wing_upper_l_twist_MSTR}
    rotation {pitch=5, roll=-5}
   }
   n10 = "Transform (Modify) Bone - Bone: dyn_wing_upper_r_twist_MSTR"!(ComponentPose=@n9, Alpha:float=1, Translation="0,0,0", Rotation="-5,0,5", Scale="1,1,1") -> (Pose) {
    rotationMode=BMM_Additive
    rotationSpace=BCS_BoneSpace
    alphaBoolBlend {blendInTime=0.1, blendOption=Linear, blendOutTime=0.1}
    boneToModify {boneName=dyn_wing_upper_r_twist_MSTR}
    rotation {pitch=-5, roll=5}
   }
   n11 = "Transform (Modify) Bone - Bone: dyn_wing_upper_l_MSTR"(ComponentPose=@n10, Alpha=1, Translation="0,-2,-1", Rotation="0,0,0", Scale="1,1,1") {
    rotationMode=BMM_Additive
    rotationSpace=BCS_BoneSpace
    translationMode=BMM_Additive
    translationSpace=BCS_BoneSpace
    alphaBoolBlend {blendInTime=0.1, blendOption=Linear, blendOutTime=0.1}
    boneToModify {boneName=dyn_wing_upper_l_MSTR}
    translation {y=-2, z=-1}
   }
   n12 = "Transform (Modify) Bone - Bone: dyn_wing_upper_r_MSTR"(ComponentPose=@n11, Alpha=1, Translation="0,2,1", Rotation="0,0,0", Scale="1,1,1") {
    rotationMode=BMM_Additive
    rotationSpace=BCS_BoneSpace
    translationMode=BMM_Additive
    translationSpace=BCS_BoneSpace
    alphaBoolBlend {blendInTime=0.1, blendOption=Linear, blendOutTime=0.1}
    boneToModify {boneName=dyn_wing_upper_r_MSTR}
    translation {y=2, z=1}
   }
   n13 = "Transform (Modify) Bone - Bone: base_apron_01_lf"!(ComponentPose=@n12, bAlphaBoolEnabled:bool=True, Translation="0,0,0", Rotation="10,0,0", Scale="1,1,1") -> (Pose) {
    alphaInputType=Bool
    rotationMode=BMM_Additive
    rotationSpace=BCS_BoneSpace
    translationMode=BMM_Additive
    translationSpace=BCS_BoneSpace
    alphaBoolBlend {blendInTime=0.5}
    boneToModify {boneName=base_apron_01_lf}
    rotation {pitch=10}
   }
   n14 = "Transform (Modify) Bone - Bone: base_apron_01_cr"!(ComponentPose=@n13, bAlphaBoolEnabled:bool=True, Translation="0,0,0", Rotation="0,10,0", Scale="1,1,1") -> (Pose) {
    alphaInputType=Bool
    rotationMode=BMM_Additive
    rotationSpace=BCS_BoneSpace
    translationMode=BMM_Additive
    translationSpace=BCS_BoneSpace
    alphaBoolBlend {blendInTime=0.5}
    boneToModify {boneName=base_apron_01_cr}
    rotation {yaw=10}
   }
   n15 = "Transform (Modify) Bone - Bone: dyn_apron_01_rt"!(ComponentPose=@n14, bAlphaBoolEnabled:bool=True, Translation="0,0,0", Rotation="-0.86,0.079,-5", Scale="1,1,1") -> (Pose) {
    alphaInputType=Bool
    rotationMode=BMM_Additive
    rotationSpace=BCS_BoneSpace
    translationMode=BMM_Additive
    translationSpace=BCS_BoneSpace
    alphaBoolBlend {blendInTime=0.5}
    boneToModify {boneName=dyn_apron_01_rt}
    rotation=<rot -0.86 0.079 -5>
   }
   n16 = "Transform (Modify) Bone - Bone: base_apron_01_rt"!(ComponentPose=@n15, bAlphaBoolEnabled:bool=True, Translation="0,0,0", Rotation="0.027,4.961,0.029", Scale="1,1,1") -> (Pose) {
    alphaInputType=Bool
    rotationMode=BMM_Additive
    rotationSpace=BCS_BoneSpace
    translationMode=BMM_Additive
    translationSpace=BCS_BoneSpace
    alphaBoolBlend {blendInTime=0.5}
    boneToModify {boneName=base_apron_01_rt}
    rotation=<rot 0.027 4.9609 0.029>
   }
   n17 = "Transform (Modify) Bone - Bone: base_apron_01_rt"(ComponentPose=@n16, bAlphaBoolEnabled=True, Translation="-2,0,0", Rotation="-5.000,-0.009,-0.002", Scale="1,1,1") {
    alphaInputType=Bool
    rotationMode=BMM_Additive
    rotationSpace=BCS_BoneSpace
    translationMode=BMM_Additive
    translationSpace=BCS_BoneSpace
    alphaBoolBlend {blendInTime=0.5, blendOutTime=0.5}
    boneToModify {boneName=base_apron_01_rt}
    rotation=<rot -5 -0.0091 -0.002>
    translation {x=-2}
   }
   n18 = "Transform (Modify) Bone - Bone: base_tail_cr"(ComponentPose=@n17, bAlphaBoolEnabled=True, Translation="0,0,0", Rotation="0,-10,0", Scale="1,1,1") {
    alphaInputType=Bool
    rotationMode=BMM_Additive
    rotationSpace=BCS_BoneSpace
    translationMode=BMM_Additive
    translationSpace=BCS_BoneSpace
    alphaBoolBlend {blendInTime=0.5, blendOutTime=0.5}
    boneToModify {boneName=base_tail_cr}
    rotation {yaw=-10}
   }
   n19 = "Transform (Modify) Bone - Bone: base_tail_left_01_cr"(ComponentPose=@n18, bAlphaBoolEnabled=True, Translation="0,0,0", Rotation="-0.000,-15.000,-0.000", Scale="1,1,1") {
    alphaInputType=Bool
    rotationMode=BMM_Additive
    rotationSpace=BCS_BoneSpace
    translationMode=BMM_Additive
    translationSpace=BCS_BoneSpace
    alphaBoolBlend {blendInTime=0.5, blendOutTime=0.5}
    boneToModify {boneName=base_tail_left_01_cr}
    rotation=<rot 0 -15 0>
   }
   n20 = "Transform (Modify) Bone - Bone: base_tail_left_02_cr"(ComponentPose=@n19, bAlphaBoolEnabled=True, Translation="0,0,0", Rotation="-0.000,-20.000,-0.000", Scale="1,1,1") {
    alphaInputType=Bool
    rotationMode=BMM_Additive
    rotationSpace=BCS_BoneSpace
    translationMode=BMM_Additive
    translationSpace=BCS_BoneSpace
    alphaBoolBlend {blendInTime=0.5, blendOutTime=0.5}
    boneToModify {boneName=base_tail_left_02_cr}
    rotation=<rot 0 -20 0>
   }
   n21 = "Transform (Modify) Bone - Bone: base_tail_right_01_cr"(ComponentPose=@n20, bAlphaBoolEnabled=True, Translation="0,0,0", Rotation="0.000,-15.000,0.000", Scale="1,1,1") {
    alphaInputType=Bool
    rotationMode=BMM_Additive
    rotationSpace=BCS_BoneSpace
    translationMode=BMM_Additive
    translationSpace=BCS_BoneSpace
    alphaBoolBlend {blendInTime=0.5, blendOutTime=0.5}
    boneToModify {boneName=base_tail_right_01_cr}
    rotation=<rot 0 -15 0>
   }
   n22 = "Transform (Modify) Bone - Bone: base_tail_right_02_cr"!(ComponentPose=@n21, bAlphaBoolEnabled:bool=True, Translation="0,0,0", Rotation="-0.000,-15.000,-0.000", Scale="1,1,1") -> (Pose) {
    alphaInputType=Bool
    rotationMode=BMM_Additive
    rotationSpace=BCS_BoneSpace
    translationMode=BMM_Additive
    translationSpace=BCS_BoneSpace
    alphaBoolBlend {blendInTime=0.5, blendOutTime=0.5}
    boneToModify {boneName=base_tail_right_02_cr}
    rotation=<rot 0 -15 0>
   }
   n23 = "Transform (Modify) Bone - Bone: dfrm_calf_r"!(ComponentPose=@n22, bAlphaBoolEnabled:bool=True, Translation="-6.056,-0.002,0.117", Rotation="0,0,0", Scale="1,1,1") -> (Pose) {
    alphaInputType=Bool
    rotationMode=BMM_Additive
    rotationSpace=BCS_BoneSpace
    translationMode=BMM_Additive
    translationSpace=BCS_BoneSpace
    alphaBoolBlend {blendInTime=0.5, blendOutTime=0.5}
    boneToModify {boneName=dfrm_calf_r}
    translation=<-6.056 -0.002 0.117>
   }
   n24 = "Transform (Modify) Bone - Bone: dyn_tail_r_MSTR"!(ComponentPose=@n23, bAlphaBoolEnabled:bool=True, Translation="0,0,0", Rotation="-19.968,4.877,-0.331", Scale="1,1,1") -> (Pose) {
    alphaInputType=Bool
    rotationMode=BMM_Additive
    rotationSpace=BCS_BoneSpace
    translationMode=BMM_Additive
    translationSpace=BCS_BoneSpace
    alphaBoolBlend {blendInTime=0.5, blendOutTime=0.5}
    boneToModify {boneName=dyn_tail_r_MSTR}
    rotation=<rot -19.968 4.8771 -0.331>
   }
   n25 = "Transform (Modify) Bone - Bone: dyn_tail_l_MSTR"!(ComponentPose=@n24, bAlphaBoolEnabled:bool=True, Translation="0,0,0", Rotation="10.079,-4.948,-0.095", Scale="1,1,1") -> (Pose) {
    alphaInputType=Bool
    rotationMode=BMM_Additive
    rotationSpace=BCS_BoneSpace
    translationMode=BMM_Additive
    translationSpace=BCS_BoneSpace
    alphaBoolBlend {blendInTime=0.5, blendOutTime=0.5}
    boneToModify {boneName=dyn_tail_l_MSTR}
    rotation=<rot 10.079 -4.948 -0.095>
   }
   n26 = "Transform (Modify) Bone - Bone: dyn_tie_lower_01_lf"!(ComponentPose=@n25, bAlphaBoolEnabled:bool=True, Translation="0,0,0", Rotation="15.163,-9.759,0.852", Scale="1,1,1") -> (Pose) {
    alphaInputType=Bool
    rotationMode=BMM_Additive
    rotationSpace=BCS_BoneSpace
    translationMode=BMM_Additive
    translationSpace=BCS_BoneSpace
    alphaBoolBlend {blendInTime=0.5, blendOutTime=0.5}
    boneToModify {boneName=dyn_tie_lower_01_lf}
    rotation=<rot 15.163 -9.7589 0.852>
   }
   n27 = "Transform (Modify) Bone - Bone: dyn_tie_lower_01_rt"!(ComponentPose=@n26, bAlphaBoolEnabled:bool=True, Translation="0,0,0", Rotation="9.827,15.237,1.390", Scale="1,1,1") -> (Pose) {
    alphaInputType=Bool
    rotationMode=BMM_Additive
    rotationSpace=BCS_BoneSpace
    translationMode=BMM_Additive
    translationSpace=BCS_BoneSpace
    alphaBoolBlend {blendInTime=0.5, blendOutTime=0.5}
    boneToModify {boneName=dyn_tie_lower_01_rt}
    rotation=<rot 9.827 15.2369 1.39>
   }
   n28 = "Transform (Modify) Bone - Bone: dyn_tie_upper_01_lf"!(ComponentPose=@n27, bAlphaBoolEnabled:bool=True, Translation="0,0,0", Rotation="10.149,-9.922,1.227", Scale="1,1,1") -> (Pose) {
    alphaInputType=Bool
    rotationMode=BMM_Additive
    rotationSpace=BCS_BoneSpace
    translationMode=BMM_Additive
    translationSpace=BCS_BoneSpace
    alphaBoolBlend {blendInTime=0.5, blendOutTime=0.5}
    boneToModify {boneName=dyn_tie_upper_01_lf}
    rotation=<rot 10.149 -9.922 1.227>
   }
   n29 = "Transform (Modify) Bone - Bone: dyn_tie_upper_01_rt"!(ComponentPose=@n28, bAlphaBoolEnabled:bool=True, Translation="0,0,0", Rotation="5.177,-10.112,0.726", Scale="1,1,1") -> (Pose) {
    alphaInputType=Bool
    rotationMode=BMM_Additive
    rotationSpace=BCS_BoneSpace
    translationMode=BMM_Additive
    translationSpace=BCS_BoneSpace
    alphaBoolBlend {blendInTime=0.5, blendOutTime=0.5}
    boneToModify {boneName=dyn_tie_upper_01_rt}
    rotation=<rot 5.177 -10.112 0.726>
   }
   n30 = "Transform (Modify) Bone - Bone: dyn_tie_lower_01_lf"(ComponentPose=@n29, bAlphaBoolEnabled=True, Translation="0,0,0", Rotation="15.163,-9.759,0.852", Scale="1,1,1") {
    alphaInputType=Bool
    rotationMode=BMM_Additive
    rotationSpace=BCS_BoneSpace
    translationMode=BMM_Additive
    translationSpace=BCS_BoneSpace
    alphaBoolBlend {blendInTime=0.5, blendOutTime=0.5}
    boneToModify {boneName=dyn_tie_lower_01_lf}
    rotation=<rot 15.163 -9.7589 0.852>
   }
   n31 = "Transform (Modify) Bone - Bone: dyn_tie_lower_01_rt"(ComponentPose=@n30, bAlphaBoolEnabled=True, Translation="0,0,0", Rotation="9.827,15.237,1.390", Scale="1,1,1") {
    alphaInputType=Bool
    rotationMode=BMM_Additive
    rotationSpace=BCS_BoneSpace
    translationMode=BMM_Additive
    translationSpace=BCS_BoneSpace
    alphaBoolBlend {blendInTime=0.5, blendOutTime=0.5}
    boneToModify {boneName=dyn_tie_lower_01_rt}
    rotation=<rot 9.827 15.2369 1.39>
   }
   n32 = "Transform (Modify) Bone - Bone: dyn_tail_l_MSTR"(ComponentPose=@n31, bAlphaBoolEnabled=True, Translation="0,0,0", Rotation="0.017,40.000,-0.002", Scale="1,1,1") {
    alphaInputType=Bool
    rotationMode=BMM_Additive
    rotationSpace=BCS_BoneSpace
    translationMode=BMM_Additive
    translationSpace=BCS_BoneSpace
    alphaBoolBlend {blendInTime=0.5, blendOutTime=0.5}
    boneToModify {boneName=dyn_tail_l_MSTR}
    rotation=<rot 0.017 40.0001 -0.002>
   }
   n33 = "Transform (Modify) Bone - Bone: dyn_tail_r_MSTR"(ComponentPose=@n32, bAlphaBoolEnabled=True, Translation="0,0,0", Rotation="0,40,0", Scale="1,1,1") {
    alphaInputType=Bool
    rotationMode=BMM_Additive
    rotationSpace=BCS_BoneSpace
    translationMode=BMM_Additive
    translationSpace=BCS_BoneSpace
    alphaBoolBlend {blendInTime=0.5, blendOutTime=0.5}
    boneToModify {boneName=dyn_tail_r_MSTR}
    rotation {yaw=40}
   }
   n34 = "Transform (Modify) Bone - Bone: base_apron_01_rt"(ComponentPose=@n33, bAlphaBoolEnabled=True, Translation="0,0,0", Rotation="0.149,-35.000,-0.015", Scale="1,1,1") {
    alphaInputType=Bool
    rotationMode=BMM_Additive
    rotationSpace=BCS_BoneSpace
    translationMode=BMM_Additive
    translationSpace=BCS_BoneSpace
    alphaBoolBlend {blendInTime=0.5, blendOutTime=0.5}
    boneToModify {boneName=base_apron_01_rt}
    rotation=<rot 0.149 -35 -0.015>
   }
   n35 = "Transform (Modify) Bone - Bone: base_apron_01_lf"(ComponentPose=@n34, bAlphaBoolEnabled=True, Translation="0,0,0", Rotation="0.055,35.000,-0.016", Scale="1,1,1") {
    alphaInputType=Bool
    rotationMode=BMM_Additive
    rotationSpace=BCS_BoneSpace
    translationMode=BMM_Additive
    translationSpace=BCS_BoneSpace
    alphaBoolBlend {blendInTime=0.5, blendOutTime=0.5}
    boneToModify {boneName=base_apron_01_lf}
    rotation=<rot 0.055 35 -0.016>
   }
   n36 = "Transform (Modify) Bone - Bone: base_tail_cr"(
    ComponentPose=@n35
    bAlphaBoolEnabled=Get IsWearingCapeFloating() -> (IsWearingCapeFloating:bool)
    Translation="0,0,0"
    Rotation="0,10,0"
    Scale="1,1,1"
   ) {
    alphaInputType=Bool
    rotationMode=BMM_Additive
    rotationSpace=BCS_BoneSpace
    translationMode=BMM_Additive
    translationSpace=BCS_BoneSpace
    alphaBoolBlend {blendInTime=0.5, blendOutTime=0.5}
    boneToModify {boneName=base_tail_cr}
    rotation {yaw=10}
   }
   Save cached pose 'PreDynamicPose'(Pose=Component To Local(ComponentPose=@n36)) {cacheName=PreDynamicPose}
   n37 = Anim Dynamics - Bone: dyn_wing_upper_r_02(ComponentPose=Local To Component(LocalPose=Use cached pose 'PreDynamicPose'()), Alpha=0.3, ExternalForce="0,0,0") {
    angularDampingOverride=0.99
    angularSpringConstant=5000
    bAngularSpring=true
    bGravityOverrideInSimSpace=true
    bOverrideAngularDamping=true
    simulationSpace=RootRelative
    alphaBoolBlend {blendOption=Linear}
    boundBone {boneName=dyn_wing_upper_r_02}
    componentAppliedLinearAccClamp=<3500 3500 3500>
    componentLinearAccScale=<0.5 0.5 0.6>
    componentLinearVelScale=<0.5 0.5 0.6>
    gravityOverride {z=-980}
    physicsBodyDefinitions [{
      sphereCollisionRadius=0
      boundBone {boneName=dyn_wing_upper_r_02}
      boxExtents=<16 16 16>
      constraintSetup {angularLimitsMax=<1 25 10>, angularLimitsMin=<-1 -10 -25>, angularTarget=<1 0 0>}
      localJointOffset=<1 0 0>
     }]
    relativeSpaceBone {boneName=root}
   }
   n38 = Anim Dynamics - Bone: dyn_wing_upper_r_03(ComponentPose=@n37, Alpha=0.3, ExternalForce="0,0,0") {
    angularDampingOverride=0.99
    angularSpringConstant=5000
    bAngularSpring=true
    bGravityOverrideInSimSpace=true
    bOverrideAngularDamping=true
    simulationSpace=RootRelative
    alphaBoolBlend {blendOption=Linear}
    boundBone {boneName=dyn_wing_upper_r_03}
    componentAppliedLinearAccClamp=<3500 3500 3500>
    componentLinearAccScale=<0.5 0.5 0.6>
    componentLinearVelScale=<0.5 0.5 0.6>
    gravityOverride {z=-980}
    physicsBodyDefinitions [{
      sphereCollisionRadius=0
      boundBone {boneName=dyn_wing_upper_r_03}
      boxExtents=<16 16 16>
      constraintSetup {angularLimitsMax=<1 25 10>, angularLimitsMin=<-1 -10 -25>, angularTarget=<1 0 0>}
      localJointOffset=<1 0 0>
     }]
    relativeSpaceBone {boneName=root}
   }
   n39 = Anim Dynamics - Bone: dyn_wing_upper_r_01(ComponentPose=@n38, Alpha=0.3, ExternalForce="0,0,0") {
    angularDampingOverride=0.99
    angularSpringConstant=5000
    bAngularSpring=true
    bGravityOverrideInSimSpace=true
    bOverrideAngularDamping=true
    simulationSpace=RootRelative
    alphaBoolBlend {blendOption=Linear}
    boundBone {boneName=dyn_wing_upper_r_01}
    componentAppliedLinearAccClamp=<3500 3500 3500>
    componentLinearAccScale=<0.5 0.5 0.6>
    componentLinearVelScale=<0.5 0.5 0.6>
    gravityOverride {z=-980}
    physicsBodyDefinitions [{
      sphereCollisionRadius=0
      boundBone {boneName=dyn_wing_upper_r_01}
      boxExtents=<16 16 16>
      constraintSetup {angularLimitsMax=<1 25 10>, angularLimitsMin=<-1 -5 -25>, angularTarget=<1 0 0>}
      localJointOffset=<1 0 0>
     }]
    relativeSpaceBone {boneName=root}
   }
   n40 = Anim Dynamics - Bone: dyn_wing_upper_r_04(ComponentPose=@n39, Alpha=0.3, ExternalForce="0,0,0") {
    angularDampingOverride=0.99
    angularSpringConstant=5000
    bAngularSpring=true
    bGravityOverrideInSimSpace=true
    bOverrideAngularDamping=true
    simulationSpace=RootRelative
    alphaBoolBlend {blendOption=Linear}
    boundBone {boneName=dyn_wing_upper_r_04}
    componentAppliedLinearAccClamp=<3500 3500 3500>
    componentLinearAccScale=<0.5 0.5 0.6>
    componentLinearVelScale=<0.5 0.5 0.6>
    gravityOverride {z=-980}
    physicsBodyDefinitions [{
      sphereCollisionRadius=0
      boundBone {boneName=dyn_wing_upper_r_04}
      boxExtents=<16 16 16>
      constraintSetup {angularLimitsMax=<1 25 10>, angularLimitsMin=<-1 0 -25>, angularTarget=<1 0 0>}
      localJointOffset=<1 0 0>
     }]
    relativeSpaceBone {boneName=root}
   }
   n41 = Anim Dynamics - Bone: dyn_wing_upper_r_05(ComponentPose=@n40, Alpha=0.3, ExternalForce="0,0,0") {
    angularDampingOverride=0.99
    angularSpringConstant=5000
    bAngularSpring=true
    bGravityOverrideInSimSpace=true
    bOverrideAngularDamping=true
    simulationSpace=RootRelative
    alphaBoolBlend {blendOption=Linear}
    boundBone {boneName=dyn_wing_upper_r_05}
    componentAppliedLinearAccClamp=<3500 3500 3500>
    componentLinearAccScale=<0.5 0.5 0.6>
    componentLinearVelScale=<0.5 0.5 0.6>
    gravityOverride {z=-980}
    physicsBodyDefinitions [{
      sphereCollisionRadius=0
      boundBone {boneName=dyn_wing_upper_r_05}
      boxExtents=<16 16 16>
      constraintSetup {angularLimitsMax=<1 25 10>, angularLimitsMin=<-1 0 -25>, angularTarget=<1 0 0>}
      localJointOffset=<1 0 0>
     }]
    relativeSpaceBone {boneName=root}
   }
   n42 = Anim Dynamics - Bone: dyn_wing_upper_l_02(ComponentPose=@n41, Alpha=0.3, ExternalForce="0,0,0") {
    angularDampingOverride=0.99
    angularSpringConstant=5000
    bAngularSpring=true
    bGravityOverrideInSimSpace=true
    bOverrideAngularDamping=true
    simulationSpace=RootRelative
    alphaBoolBlend {blendOption=Linear}
    boundBone {boneName=dyn_wing_upper_l_02}
    componentAppliedLinearAccClamp=<3500 3500 3500>
    componentLinearAccScale=<0.5 0.5 0.6>
    componentLinearVelScale=<0.5 0.5 0.6>
    gravityOverride {z=-980}
    physicsBodyDefinitions [{
      sphereCollisionRadius=0
      boundBone {boneName=dyn_wing_upper_l_02}
      boxExtents=<16 16 16>
      constraintSetup {angularLimitsMax=<1 25 10>, angularLimitsMin=<-1 -10 -25>, angularTarget=<1 0 0>}
      localJointOffset=<-1 0 0>
     }]
    relativeSpaceBone {boneName=root}
   }
   n43 = Anim Dynamics - Bone: dyn_wing_upper_l_03(ComponentPose=@n42, Alpha=0.3, ExternalForce="0,0,0") {
    angularDampingOverride=0.99
    angularSpringConstant=5000
    bAngularSpring=true
    bGravityOverrideInSimSpace=true
    bOverrideAngularDamping=true
    simulationSpace=RootRelative
    alphaBoolBlend {blendOption=Linear}
    boundBone {boneName=dyn_wing_upper_l_03}
    componentAppliedLinearAccClamp=<3500 3500 3500>
    componentLinearAccScale=<0.5 0.5 0.6>
    componentLinearVelScale=<0.5 0.5 0.6>
    gravityOverride {z=-980}
    physicsBodyDefinitions [{
      sphereCollisionRadius=0
      boundBone {boneName=dyn_wing_upper_l_03}
      boxExtents=<16 16 16>
      constraintSetup {angularLimitsMax=<1 25 10>, angularLimitsMin=<-1 -10 -25>, angularTarget=<1 0 0>}
      localJointOffset=<-1 0 0>
     }]
    relativeSpaceBone {boneName=root}
   }
   n44 = Anim Dynamics - Bone: dyn_wing_upper_l_01(ComponentPose=@n43, Alpha=0.3, ExternalForce="0,0,0") {
    angularDampingOverride=0.99
    angularSpringConstant=5000
    bAngularSpring=true
    bGravityOverrideInSimSpace=true
    bOverrideAngularDamping=true
    simulationSpace=RootRelative
    alphaBoolBlend {blendOption=Linear}
    boundBone {boneName=dyn_wing_upper_l_01}
    componentAppliedLinearAccClamp=<3500 3500 3500>
    componentLinearAccScale=<0.5 0.5 0.6>
    componentLinearVelScale=<0.5 0.5 0.6>
    gravityOverride {z=-980}
    physicsBodyDefinitions [{
      sphereCollisionRadius=0
      boundBone {boneName=dyn_wing_upper_l_01}
      boxExtents=<16 16 16>
      constraintSetup {angularLimitsMax=<1 25 10>, angularLimitsMin=<-1 -5 -25>, angularTarget=<1 0 0>}
      localJointOffset=<-1 0 0>
     }]
    relativeSpaceBone {boneName=root}
   }
   n45 = Anim Dynamics - Bone: dyn_wing_upper_l_04(ComponentPose=@n44, Alpha=0.3, ExternalForce="0,0,0") {
    angularDampingOverride=0.99
    angularSpringConstant=5000
    bAngularSpring=true
    bGravityOverrideInSimSpace=true
    bOverrideAngularDamping=true
    simulationSpace=RootRelative
    alphaBoolBlend {blendOption=Linear}
    boundBone {boneName=dyn_wing_upper_l_04}
    componentAppliedLinearAccClamp=<3500 3500 3500>
    componentLinearAccScale=<0.5 0.5 0.6>
    componentLinearVelScale=<0.5 0.5 0.6>
    gravityOverride {z=-980}
    physicsBodyDefinitions [{
      sphereCollisionRadius=0
      boundBone {boneName=dyn_wing_upper_l_04}
      boxExtents=<16 16 16>
      constraintSetup {angularLimitsMax=<1 25 10>, angularLimitsMin=<-1 0 -25>, angularTarget=<1 0 0>}
      localJointOffset=<-1 0 0>
     }]
    relativeSpaceBone {boneName=root}
   }
   n46 = Anim Dynamics - Bone: dyn_wing_upper_l_05(ComponentPose=@n45, Alpha=0.3, ExternalForce="0,0,0") {
    angularDampingOverride=0.99
    angularSpringConstant=5000
    bAngularSpring=true
    bGravityOverrideInSimSpace=true
    bOverrideAngularDamping=true
    simulationSpace=RootRelative
    alphaBoolBlend {blendOption=Linear}
    boundBone {boneName=dyn_wing_upper_l_05}
    componentAppliedLinearAccClamp=<3500 3500 3500>
    componentLinearAccScale=<0.5 0.5 0.6>
    componentLinearVelScale=<0.5 0.5 0.6>
    gravityOverride {z=-980}
    physicsBodyDefinitions [{
      sphereCollisionRadius=0
      boundBone {boneName=dyn_wing_upper_l_05}
      boxExtents=<16 16 16>
      constraintSetup {angularLimitsMax=<1 25 10>, angularLimitsMin=<-1 0 -25>, angularTarget=<1 0 0>}
      localJointOffset=<-1 0 0>
     }]
    relativeSpaceBone {boneName=root}
   }
   n47 = Anim Dynamics - Bone: dyn_calf_feather_01(ComponentPose=@n46, Alpha=0.5, ExternalForce="0,0,0") {
    angularDampingOverride=0.9999
    angularSpringConstant=75
    bAngularSpring=true
    bGravityOverrideInSimSpace=true
    bOverrideAngularDamping=true
    gravityScale=0.8
    simulationSpace=BoneRelative
    alphaBoolBlend {blendOption=Linear}
    boundBone {boneName=dyn_calf_feather_01}
    componentAppliedLinearAccClamp=<3500 3500 3500>
    componentLinearAccScale=<1.5 1.5 1.25>
    componentLinearVelScale=<1.5 1.5 1.25>
    gravityOverride {z=-980}
    physicsBodyDefinitions [{
      sphereCollisionRadius=0
      boundBone {boneName=dyn_calf_feather_01}
      boxExtents=<17 17 10>
      constraintSetup {angularLimitsMax=<5 3 15>, angularLimitsMin=<-5 -15 -15>, angularTarget=<1 0 0>}
      localJointOffset=<3 0 0>
     }]
    relativeSpaceBone {boneName=root}
   }
   n48 = Anim Dynamics - Bone: dyn_calf_feather_02(ComponentPose=@n47, Alpha=0.5, ExternalForce="0,0,0") {
    angularDampingOverride=0.9999
    angularSpringConstant=75
    bAngularSpring=true
    bGravityOverrideInSimSpace=true
    bOverrideAngularDamping=true
    gravityScale=0.8
    simulationSpace=BoneRelative
    alphaBoolBlend {blendOption=Linear}
    boundBone {boneName=dyn_calf_feather_02}
    componentAppliedLinearAccClamp=<3500 3500 3500>
    componentLinearAccScale=<1.5 1.5 1.25>
    componentLinearVelScale=<1.5 1.5 1.25>
    gravityOverride {z=-980}
    physicsBodyDefinitions [{
      sphereCollisionRadius=0
      boundBone {boneName=dyn_calf_feather_02}
      boxExtents=<17 17 10>
      constraintSetup {angularLimitsMax=<5 3 15>, angularLimitsMin=<-5 -15 -15>, angularTarget=<1 0 0>}
      localJointOffset=<4 0.5 0>
     }]
    relativeSpaceBone {boneName=root}
   }
   n49 = Anim Dynamics - Bone: dyn_calf_feather_03(ComponentPose=@n48, Alpha=0.5, ExternalForce="0,0,0") {
    angularDampingOverride=0.9999
    angularSpringConstant=75
    bAngularSpring=true
    bGravityOverrideInSimSpace=true
    bOverrideAngularDamping=true
    gravityScale=0.8
    simulationSpace=BoneRelative
    alphaBoolBlend {blendOption=Linear}
    boundBone {boneName=dyn_calf_feather_03}
    componentAppliedLinearAccClamp=<3500 3500 3500>
    componentLinearAccScale=<1.5 1.5 1.25>
    componentLinearVelScale=<1.5 1.5 1.25>
    gravityOverride {z=-980}
    physicsBodyDefinitions [{
      sphereCollisionRadius=0
      boundBone {boneName=dyn_calf_feather_03}
      boxExtents=<17 17 10>
      constraintSetup {angularLimitsMax=<5 3 15>, angularLimitsMin=<-5 -15 -15>, angularTarget=<1 0 0>}
      localJointOffset=<3 1 0>
     }]
    relativeSpaceBone {boneName=root}
   }
   n50 = Anim Dynamics - Bone: dyn_calf_feather_04(ComponentPose=@n49, Alpha=0.5, ExternalForce="0,0,0") {
    angularDampingOverride=0.9999
    angularSpringConstant=75
    bAngularSpring=true
    bGravityOverrideInSimSpace=true
    bOverrideAngularDamping=true
    gravityScale=0.8
    simulationSpace=BoneRelative
    alphaBoolBlend {blendOption=Linear}
    boundBone {boneName=dyn_calf_feather_04}
    componentAppliedLinearAccClamp=<3500 3500 3500>
    componentLinearAccScale=<1.5 1.5 1.25>
    componentLinearVelScale=<1.5 1.5 1.25>
    gravityOverride {z=-980}
    physicsBodyDefinitions [{
      sphereCollisionRadius=0
      boundBone {boneName=dyn_calf_feather_04}
      boxExtents=<17 17 10>
      constraintSetup {angularLimitsMax=<5 3 15>, angularLimitsMin=<-5 -15 -15>, angularTarget=<1 0 0>}
      localJointOffset=<4 0 0>
     }]
    relativeSpaceBone {boneName=root}
   }
   n51 = Anim Dynamics - Bone: dyn_calf_feather_05(ComponentPose=@n50, Alpha=0.5, ExternalForce="0,0,0") {
    angularDampingOverride=0.9999
    angularSpringConstant=75
    bAngularSpring=true
    bGravityOverrideInSimSpace=true
    bOverrideAngularDamping=true
    gravityScale=0.8
    simulationSpace=BoneRelative
    alphaBoolBlend {blendOption=Linear}
    boundBone {boneName=dyn_calf_feather_05}
    componentAppliedLinearAccClamp=<3500 3500 3500>
    componentLinearAccScale=<1.5 1.5 1.25>
    componentLinearVelScale=<1.5 1.5 1.25>
    gravityOverride {z=-980}
    physicsBodyDefinitions [{
      sphereCollisionRadius=0
      boundBone {boneName=dyn_calf_feather_05}
      boxExtents=<17 17 10>
      constraintSetup {angularLimitsMax=<5 3 15>, angularLimitsMin=<-5 -15 -15>, angularTarget=<1 0 0>}
      localJointOffset=<3 0 0>
     }]
    relativeSpaceBone {boneName=root}
   }
   n52 = Anim Dynamics - Bone: dyn_calf_feather_06(ComponentPose=@n51, Alpha=0.5, ExternalForce="0,0,0") {
    angularDampingOverride=0.9999
    angularSpringConstant=75
    bAngularSpring=true
    bGravityOverrideInSimSpace=true
    bOverrideAngularDamping=true
    gravityScale=0.8
    simulationSpace=BoneRelative
    alphaBoolBlend {blendOption=Linear}
    boundBone {boneName=dyn_calf_feather_06}
    componentAppliedLinearAccClamp=<3500 3500 3500>
    componentLinearAccScale=<1.5 1.5 1.25>
    componentLinearVelScale=<1.5 1.5 1.25>
    gravityOverride {z=-980}
    physicsBodyDefinitions [{
      sphereCollisionRadius=0
      boundBone {boneName=dyn_calf_feather_06}
      boxExtents=<17 17 10>
      constraintSetup {angularLimitsMax=<5 3 15>, angularLimitsMin=<-5 -15 -15>, angularTarget=<1 0 0>}
      localJointOffset=<4 0.5 0>
     }]
    relativeSpaceBone {boneName=root}
   }
   n53 = Anim Dynamics - Bone: dyn_calf_feather_07(ComponentPose=@n52, Alpha=0.5, ExternalForce="0,0,0") {
    angularDampingOverride=0.9999
    angularSpringConstant=75
    bAngularSpring=true
    bGravityOverrideInSimSpace=true
    bOverrideAngularDamping=true
    gravityScale=0.8
    simulationSpace=BoneRelative
    alphaBoolBlend {blendOption=Linear}
    boundBone {boneName=dyn_calf_feather_07}
    componentAppliedLinearAccClamp=<3500 3500 3500>
    componentLinearAccScale=<1.5 1.5 1.25>
    componentLinearVelScale=<1.5 1.5 1.25>
    gravityOverride {z=-980}
    physicsBodyDefinitions [{
      sphereCollisionRadius=0
      boundBone {boneName=dyn_calf_feather_07}
      boxExtents=<17 17 10>
      constraintSetup {angularLimitsMax=<5 3 15>, angularLimitsMin=<-5 -15 -15>, angularTarget=<1 0 0>}
      localJointOffset=<3 1 0>
     }]
    relativeSpaceBone {boneName=root}
   }
   n54 = Anim Dynamics - Bone: dyn_calf_feather_08(ComponentPose=@n53, Alpha=0.5, ExternalForce="0,0,0") {
    angularDampingOverride=0.9999
    angularSpringConstant=75
    bAngularSpring=true
    bGravityOverrideInSimSpace=true
    bOverrideAngularDamping=true
    gravityScale=0.8
    simulationSpace=BoneRelative
    alphaBoolBlend {blendOption=Linear}
    boundBone {boneName=dyn_calf_feather_08}
    componentAppliedLinearAccClamp=<3500 3500 3500>
    componentLinearAccScale=<1.5 1.5 1.25>
    componentLinearVelScale=<1.5 1.5 1.25>
    gravityOverride {z=-980}
    physicsBodyDefinitions [{
      sphereCollisionRadius=0
      boundBone {boneName=dyn_calf_feather_08}
      boxExtents=<17 17 10>
      constraintSetup {angularLimitsMax=<5 3 15>, angularLimitsMin=<-5 -15 -15>, angularTarget=<1 0 0>}
      localJointOffset=<4 0 0>
     }]
    relativeSpaceBone {boneName=root}
   }
   n55 = Anim Dynamics - Bone: dyn_calf_feather_l_01(ComponentPose=@n54, Alpha=0.5, ExternalForce="0,0,0") {
    angularDampingOverride=0.9999
    angularSpringConstant=75
    bAngularSpring=true
    bGravityOverrideInSimSpace=true
    bOverrideAngularDamping=true
    gravityScale=0.8
    simulationSpace=BoneRelative
    alphaBoolBlend {blendOption=Linear}
    boundBone {boneName=dyn_calf_feather_l_01}
    componentAppliedLinearAccClamp=<3500 3500 3500>
    componentLinearAccScale=<1.5 1.5 1.25>
    componentLinearVelScale=<1.5 1.5 1.25>
    gravityOverride {z=-980}
    physicsBodyDefinitions [{
      sphereCollisionRadius=0
      boundBone {boneName=dyn_calf_feather_l_01}
      boxExtents=<17 17 10>
      constraintSetup {angularLimitsMax=<5 3 15>, angularLimitsMin=<-5 -15 -15>, angularTarget=<1 0 0>}
      localJointOffset=<3 -0.5 0>
     }]
    relativeSpaceBone {boneName=root}
   }
   n56 = Anim Dynamics - Bone: dyn_calf_feather_l_02(ComponentPose=@n55, Alpha=0.5, ExternalForce="0,0,0") {
    angularDampingOverride=0.9999
    angularSpringConstant=75
    bAngularSpring=true
    bGravityOverrideInSimSpace=true
    bOverrideAngularDamping=true
    gravityScale=0.8
    simulationSpace=BoneRelative
    alphaBoolBlend {blendOption=Linear}
    boundBone {boneName=dyn_calf_feather_l_02}
    componentAppliedLinearAccClamp=<3500 3500 3500>
    componentLinearAccScale=<1.5 1.5 1.25>
    componentLinearVelScale=<1.5 1.5 1.25>
    gravityOverride {z=-980}
    physicsBodyDefinitions [{
      sphereCollisionRadius=0
      boundBone {boneName=dyn_calf_feather_l_02}
      boxExtents=<17 17 10>
      constraintSetup {angularLimitsMax=<5 3 15>, angularLimitsMin=<-5 -15 -15>, angularTarget=<1 0 0>}
      localJointOffset=<4 0.5 0>
     }]
    relativeSpaceBone {boneName=root}
   }
   n57 = Anim Dynamics - Bone: dyn_calf_feather_l_03(ComponentPose=@n56, Alpha=0.5, ExternalForce="0,0,0") {
    angularDampingOverride=0.9999
    angularSpringConstant=75
    bAngularSpring=true
    bGravityOverrideInSimSpace=true
    bOverrideAngularDamping=true
    gravityScale=0.8
    simulationSpace=BoneRelative
    alphaBoolBlend {blendOption=Linear}
    boundBone {boneName=dyn_calf_feather_l_03}
    componentAppliedLinearAccClamp=<3500 3500 3500>
    componentLinearAccScale=<1.5 1.5 1.25>
    componentLinearVelScale=<1.5 1.5 1.25>
    gravityOverride {z=-980}
    physicsBodyDefinitions [{
      sphereCollisionRadius=0
      boundBone {boneName=dyn_calf_feather_l_03}
      boxExtents=<17 17 10>
      constraintSetup {angularLimitsMax=<5 3 15>, angularLimitsMin=<-5 -15 -15>, angularTarget=<1 0 0>}
      localJointOffset=<3 1 0>
     }]
    relativeSpaceBone {boneName=root}
   }
   n58 = Anim Dynamics - Bone: dyn_calf_feather_l_04(ComponentPose=@n57, Alpha=0.5, ExternalForce="0,0,0") {
    angularDampingOverride=0.9999
    angularSpringConstant=75
    bAngularSpring=true
    bGravityOverrideInSimSpace=true
    bOverrideAngularDamping=true
    gravityScale=0.8
    simulationSpace=BoneRelative
    alphaBoolBlend {blendOption=Linear}
    boundBone {boneName=dyn_calf_feather_l_04}
    componentAppliedLinearAccClamp=<3500 3500 3500>
    componentLinearAccScale=<1.5 1.5 1.25>
    componentLinearVelScale=<1.5 1.5 1.25>
    gravityOverride {z=-980}
    physicsBodyDefinitions [{
      sphereCollisionRadius=0
      boundBone {boneName=dyn_calf_feather_l_04}
      boxExtents=<17 17 10>
      constraintSetup {angularLimitsMax=<5 3 15>, angularLimitsMin=<-5 -15 -15>, angularTarget=<1 0 0>}
      localJointOffset=<4 0 0>
     }]
    relativeSpaceBone {boneName=root}
   }
   n59 = Anim Dynamics - Bone: dyn_calf_feather_l_05(ComponentPose=@n58, Alpha=0.5, ExternalForce="0,0,0") {
    angularDampingOverride=0.9999
    angularSpringConstant=75
    bAngularSpring=true
    bGravityOverrideInSimSpace=true
    bOverrideAngularDamping=true
    gravityScale=0.8
    simulationSpace=BoneRelative
    alphaBoolBlend {blendOption=Linear}
    boundBone {boneName=dyn_calf_feather_l_05}
    componentAppliedLinearAccClamp=<3500 3500 3500>
    componentLinearAccScale=<1.5 1.5 1.25>
    componentLinearVelScale=<1.5 1.5 1.25>
    gravityOverride {z=-980}
    physicsBodyDefinitions [{
      sphereCollisionRadius=0
      boundBone {boneName=dyn_calf_feather_l_05}
      boxExtents=<17 17 10>
      constraintSetup {angularLimitsMax=<5 3 15>, angularLimitsMin=<-5 -15 -15>, angularTarget=<1 0 0>}
      localJointOffset=<3 -0.5 0>
     }]
    relativeSpaceBone {boneName=root}
   }
   n60 = Anim Dynamics - Bone: dyn_calf_feather_l_06(ComponentPose=@n59, Alpha=0.5, ExternalForce="0,0,0") {
    angularDampingOverride=0.9999
    angularSpringConstant=75
    bAngularSpring=true
    bGravityOverrideInSimSpace=true
    bOverrideAngularDamping=true
    gravityScale=0.8
    simulationSpace=BoneRelative
    alphaBoolBlend {blendOption=Linear}
    boundBone {boneName=dyn_calf_feather_l_06}
    componentAppliedLinearAccClamp=<3500 3500 3500>
    componentLinearAccScale=<1.5 1.5 1.25>
    componentLinearVelScale=<1.5 1.5 1.25>
    gravityOverride {z=-980}
    physicsBodyDefinitions [{
      sphereCollisionRadius=0
      boundBone {boneName=dyn_calf_feather_l_06}
      boxExtents=<17 17 10>
      constraintSetup {angularLimitsMax=<5 3 15>, angularLimitsMin=<-5 -15 -15>, angularTarget=<1 0 0>}
      localJointOffset=<4 0.5 0>
     }]
    relativeSpaceBone {boneName=root}
   }
   n61 = Anim Dynamics - Bone: dyn_calf_feather_l_07(ComponentPose=@n60, Alpha=0.5, ExternalForce="0,0,0") {
    angularDampingOverride=0.9999
    angularSpringConstant=75
    bAngularSpring=true
    bGravityOverrideInSimSpace=true
    bOverrideAngularDamping=true
    gravityScale=0.8
    simulationSpace=BoneRelative
    alphaBoolBlend {blendOption=Linear}
    boundBone {boneName=dyn_calf_feather_l_07}
    componentAppliedLinearAccClamp=<3500 3500 3500>
    componentLinearAccScale=<1.5 1.5 1.25>
    componentLinearVelScale=<1.5 1.5 1.25>
    gravityOverride {z=-980}
    physicsBodyDefinitions [{
      sphereCollisionRadius=0
      boundBone {boneName=dyn_calf_feather_l_07}
      boxExtents=<17 17 10>
      constraintSetup {angularLimitsMax=<5 3 15>, angularLimitsMin=<-5 -15 -15>, angularTarget=<1 0 0>}
      localJointOffset=<3 1 0>
     }]
    relativeSpaceBone {boneName=root}
   }
   n62 = Anim Dynamics - Bone: dyn_calf_feather_l_08(ComponentPose=@n61, Alpha=0.5, ExternalForce="0,0,0") {
    angularDampingOverride=0.9999
    angularSpringConstant=75
    bAngularSpring=true
    bGravityOverrideInSimSpace=true
    bOverrideAngularDamping=true
    gravityScale=0.8
    simulationSpace=BoneRelative
    alphaBoolBlend {blendOption=Linear}
    boundBone {boneName=dyn_calf_feather_l_08}
    componentAppliedLinearAccClamp=<3500 3500 3500>
    componentLinearAccScale=<1.5 1.5 1.25>
    componentLinearVelScale=<1.5 1.5 1.25>
    gravityOverride {z=-980}
    physicsBodyDefinitions [{
      sphereCollisionRadius=0
      boundBone {boneName=dyn_calf_feather_l_08}
      boxExtents=<17 17 10>
      constraintSetup {angularLimitsMax=<5 3 15>, angularLimitsMin=<-5 -15 -15>, angularTarget=<1 0 0>}
      localJointOffset=<4 0 0>
     }]
    relativeSpaceBone {boneName=root}
   }
   n63 = Anim Dynamics - Bone: dyn_calf_feather_l_09(ComponentPose=@n62, Alpha=0.5, ExternalForce="0,0,0") {
    angularDampingOverride=0.9999
    angularSpringConstant=75
    bAngularSpring=true
    bGravityOverrideInSimSpace=true
    bOverrideAngularDamping=true
    gravityScale=0.8
    simulationSpace=BoneRelative
    alphaBoolBlend {blendOption=Linear}
    boundBone {boneName=dyn_calf_feather_l_09}
    componentAppliedLinearAccClamp=<3500 3500 3500>
    componentLinearAccScale=<1.5 1.5 1.25>
    componentLinearVelScale=<1.5 1.5 1.25>
    gravityOverride {z=-980}
    physicsBodyDefinitions [{
      sphereCollisionRadius=0
      boundBone {boneName=dyn_calf_feather_l_09}
      boxExtents=<17 17 10>
      constraintSetup {angularLimitsMax=<5 3 15>, angularLimitsMin=<-5 -15 -15>, angularTarget=<1 0 0>}
      localJointOffset=<3 1 0>
     }]
    relativeSpaceBone {boneName=root}
   }
   n64 = Anim Dynamics - Bone: dyn_calf_feather_l_10(ComponentPose=@n63, Alpha=0.5, ExternalForce="0,0,0") {
    angularDampingOverride=0.9999
    angularSpringConstant=75
    bAngularSpring=true
    bGravityOverrideInSimSpace=true
    bOverrideAngularDamping=true
    gravityScale=0.8
    simulationSpace=BoneRelative
    alphaBoolBlend {blendOption=Linear}
    boundBone {boneName=dyn_calf_feather_l_10}
    componentAppliedLinearAccClamp=<3500 3500 3500>
    componentLinearAccScale=<1.5 1.5 1.25>
    componentLinearVelScale=<1.5 1.5 1.25>
    gravityOverride {z=-980}
    physicsBodyDefinitions [{
      sphereCollisionRadius=0
      boundBone {boneName=dyn_calf_feather_l_10}
      boxExtents=<17 17 10>
      constraintSetup {angularLimitsMax=<5 3 15>, angularLimitsMin=<-5 -15 -15>, angularTarget=<1 0 0>}
      localJointOffset=<4 0 0>
     }]
    relativeSpaceBone {boneName=root}
   }
   n65 = Anim Dynamics - Bone: dyn_tail_l_01(ComponentPose=@n64, Alpha=1, ExternalForce="0,0,0") {
    angularDampingOverride=1
    angularSpringConstant=20000
    bAngularSpring=true
    bOverrideAngularDamping=true
    bUseGravityOverride=true
    gravityScale=0.35
    simulationSpace=RootRelative
    boundBone {boneName=dyn_tail_l_01}
    componentAppliedLinearAccClamp=<3000 3000 3000>
    componentLinearAccScale=<1 1 1>
    componentLinearVelScale=<1 1 1>
    gravityOverride {z=-980}
    physicsBodyDefinitions [{
      boundBone {boneName=dyn_tail_l_01}
      boxExtents=<50 50 50>
      constraintSetup {angularLimitsMax=<0 8 0>, angularLimitsMin=<0 -8 -15>, angularTarget=<1 0 0>}
      localJointOffset=<6 0 0>
     }]
    simSpaceSettings {maxAngularAcceleration=27, maxAngularVelocity=0.3, simSpaceAngularAlpha=1}
   }
   n66 = Anim Dynamics - Bone: dyn_tail_l_02(ComponentPose=@n65, Alpha=1, ExternalForce="0,0,0") {
    angularDampingOverride=1
    angularSpringConstant=20000
    bAngularSpring=true
    bOverrideAngularDamping=true
    bUseGravityOverride=true
    gravityScale=0.35
    simulationSpace=RootRelative
    boundBone {boneName=dyn_tail_l_02}
    componentAppliedLinearAccClamp=<3000 3000 3000>
    componentLinearAccScale=<1 1 1>
    componentLinearVelScale=<1 1 1>
    gravityOverride {z=-980}
    physicsBodyDefinitions [{
      boundBone {boneName=dyn_tail_l_02}
      boxExtents=<50 50 50>
      constraintSetup {angularLimitsMax=<0 8 0>, angularLimitsMin=<0 -8 -15>, angularTarget=<1 0 0>}
      localJointOffset=<6 0 0>
     }]
    simSpaceSettings {maxAngularAcceleration=27, maxAngularVelocity=0.3, simSpaceAngularAlpha=1}
   }
   n67 = Anim Dynamics - Bone: dyn_tail_l_03(ComponentPose=@n66, Alpha=1, ExternalForce="0,0,0") {
    angularDampingOverride=1
    angularSpringConstant=20000
    bAngularSpring=true
    bOverrideAngularDamping=true
    bUseGravityOverride=true
    gravityScale=0.35
    simulationSpace=RootRelative
    boundBone {boneName=dyn_tail_l_03}
    componentAppliedLinearAccClamp=<3000 3000 3000>
    componentLinearAccScale=<1 1 1>
    componentLinearVelScale=<1 1 1>
    gravityOverride {z=-980}
    physicsBodyDefinitions [{
      boundBone {boneName=dyn_tail_l_03}
      boxExtents=<50 50 50>
      constraintSetup {angularLimitsMax=<0 8 0>, angularLimitsMin=<0 -8 -15>, angularTarget=<1 0 0>}
      localJointOffset=<6 0 0>
     }]
    simSpaceSettings {maxAngularAcceleration=27, maxAngularVelocity=0.3, simSpaceAngularAlpha=1}
   }
   n68 = Anim Dynamics - Bone: dyn_tail_r_01(ComponentPose=@n67, Alpha=1, ExternalForce="0,0,0") {
    angularDampingOverride=1
    angularSpringConstant=20000
    bAngularSpring=true
    bOverrideAngularDamping=true
    bUseGravityOverride=true
    gravityScale=0.35
    simulationSpace=RootRelative
    boundBone {boneName=dyn_tail_r_01}
    componentAppliedLinearAccClamp=<3000 3000 3000>
    componentLinearAccScale=<1 1 1>
    componentLinearVelScale=<1 1 1>
    gravityOverride {z=-980}
    physicsBodyDefinitions [{
      boundBone {boneName=dyn_tail_r_01}
      boxExtents=<50 50 50>
      constraintSetup {angularLimitsMax=<0 8 0>, angularLimitsMin=<0 -8 -15>, angularTarget=<1 0 0>}
      localJointOffset=<6 0 0>
     }]
    simSpaceSettings {maxAngularAcceleration=27, maxAngularVelocity=0.3, simSpaceAngularAlpha=1}
   }
   n69 = Anim Dynamics - Bone: dyn_tail_r_02(ComponentPose=@n68, Alpha=1, ExternalForce="0,0,0") {
    angularDampingOverride=1
    angularSpringConstant=20000
    bAngularSpring=true
    bOverrideAngularDamping=true
    bUseGravityOverride=true
    gravityScale=0.35
    simulationSpace=RootRelative
    boundBone {boneName=dyn_tail_r_02}
    componentAppliedLinearAccClamp=<3000 3000 3000>
    componentLinearAccScale=<1 1 1>
    componentLinearVelScale=<1 1 1>
    gravityOverride {z=-980}
    physicsBodyDefinitions [{
      boundBone {boneName=dyn_tail_r_02}
      boxExtents=<50 50 50>
      constraintSetup {angularLimitsMax=<0 8 0>, angularLimitsMin=<0 -8 -15>, angularTarget=<1 0 0>}
      localJointOffset=<6 0 0>
     }]
    simSpaceSettings {maxAngularAcceleration=27, maxAngularVelocity=0.3, simSpaceAngularAlpha=1}
   }
   n70 = Anim Dynamics - Bone: dyn_tail_r_03(ComponentPose=@n69, Alpha=1, ExternalForce="0,0,0") {
    angularDampingOverride=1
    angularSpringConstant=20000
    bAngularSpring=true
    bOverrideAngularDamping=true
    bUseGravityOverride=true
    gravityScale=0.35
    simulationSpace=RootRelative
    boundBone {boneName=dyn_tail_r_03}
    componentAppliedLinearAccClamp=<3000 3000 3000>
    componentLinearAccScale=<1 1 1>
    componentLinearVelScale=<1 1 1>
    gravityOverride {z=-980}
    physicsBodyDefinitions [{
      boundBone {boneName=dyn_tail_r_03}
      boxExtents=<50 50 50>
      constraintSetup {angularLimitsMax=<0 8 0>, angularLimitsMin=<0 -8 -15>, angularTarget=<1 0 0>}
      localJointOffset=<6 0 0>
     }]
    simSpaceSettings {maxAngularAcceleration=27, maxAngularVelocity=0.3, simSpaceAngularAlpha=1}
   }
   n71 = Anim Dynamics - Bone: dyn_tail_left_01_01_cr!(ComponentPose=@n70, Alpha=1, ComponentLinearAccScale:Vector="1,1,1", ComponentLinearVelScale:Vector="1,1,1", ExternalForce="0,0,0") -> (Pose) {
    angularDampingOverride=1
    angularSpringConstant=20000
    bAngularSpring=true
    bOverrideAngularDamping=true
    bUseGravityOverride=true
    gravityScale=0.35
    simulationSpace=RootRelative
    boundBone {boneName=dyn_tail_left_01_01_cr}
    componentAppliedLinearAccClamp {x=3000, z=3000}
    componentLinearAccScale=<1 1 1>
    componentLinearVelScale=<1 1 1>
    gravityOverride {z=-980}
    physicsBodyDefinitions [{
      boundBone {boneName=dyn_tail_left_01_01_cr}
      boxExtents=<20 20 20>
      constraintSetup {angularLimitsMax=<0 5 5>, angularLimitsMin=<0 -5 -20>, angularTarget=<1 0 0>}
      localJointOffset=<4 0 0>
     }]
    simSpaceSettings {maxAngularAcceleration=35, maxAngularVelocity=8, simSpaceAngularAlpha=1}
   }
   n72 = Anim Dynamics - Bone: dyn_tail_left_01_02_cr!(ComponentPose=@n71, Alpha=1, ComponentLinearAccScale:Vector="1,1,1", ComponentLinearVelScale:Vector="1,1,1", ExternalForce="0,0,0") -> (Pose) {
    angularDampingOverride=1
    angularSpringConstant=20000
    bAngularSpring=true
    bOverrideAngularDamping=true
    bUseGravityOverride=true
    gravityScale=0.35
    simulationSpace=RootRelative
    boundBone {boneName=dyn_tail_left_01_02_cr}
    componentAppliedLinearAccClamp {x=3000, z=3000}
    componentLinearAccScale=<1 1 1>
    componentLinearVelScale=<1 1 1>
    gravityOverride {z=-980}
    physicsBodyDefinitions [{
      boundBone {boneName=dyn_tail_left_01_02_cr}
      boxExtents=<40 40 40>
      constraintSetup {angularLimitsMax=<0 5 5>, angularLimitsMin=<0 -5 -15>, angularTarget=<1 0 0>}
      localJointOffset=<4 0 0>
     }]
    simSpaceSettings {maxAngularAcceleration=35, maxAngularVelocity=8, simSpaceAngularAlpha=1}
   }
   n73 = Anim Dynamics - Bone: dyn_tail_left_01_03_cr!(ComponentPose=@n72, Alpha=1, ComponentLinearAccScale:Vector="1,1,1", ComponentLinearVelScale:Vector="1,1,1", ExternalForce="0,0,0") -> (Pose) {
    angularDampingOverride=1
    angularSpringConstant=20000
    bAngularSpring=true
    bOverrideAngularDamping=true
    bUseGravityOverride=true
    gravityScale=0.35
    simulationSpace=RootRelative
    boundBone {boneName=dyn_tail_left_01_03_cr}
    componentAppliedLinearAccClamp {x=3000, z=3000}
    componentLinearAccScale=<1 1 1>
    componentLinearVelScale=<1 1 1>
    gravityOverride {z=-980}
    physicsBodyDefinitions [{
      boundBone {boneName=dyn_tail_left_01_03_cr}
      boxExtents=<40 40 40>
      constraintSetup {angularLimitsMax=<0 5 5>, angularLimitsMin=<0 -5 -15>, angularTarget=<1 0 0>}
      localJointOffset=<4 0 0>
     }]
    simSpaceSettings {maxAngularAcceleration=35, maxAngularVelocity=8, simSpaceAngularAlpha=1}
   }
   n74 = Anim Dynamics - Bone: dyn_tail_left_02_01_cr!(ComponentPose=@n73, Alpha=1, ComponentLinearAccScale:Vector="1,1,1", ComponentLinearVelScale:Vector="1,1,1", ExternalForce="0,0,0") -> (Pose) {
    angularDampingOverride=1
    angularSpringConstant=20000
    bAngularSpring=true
    bOverrideAngularDamping=true
    bUseGravityOverride=true
    gravityScale=0.35
    simulationSpace=RootRelative
    boundBone {boneName=dyn_tail_left_02_01_cr}
    componentAppliedLinearAccClamp {x=3000, z=3000}
    componentLinearAccScale=<1 1 1>
    componentLinearVelScale=<1 1 1>
    gravityOverride {z=-980}
    physicsBodyDefinitions [{
      boundBone {boneName=dyn_tail_left_02_01_cr}
      boxExtents=<40 40 40>
      constraintSetup {angularLimitsMax=<0 5 5>, angularLimitsMin=<0 -5 -15>, angularTarget=<1 0 0>}
      localJointOffset=<4 0 0>
     }]
    simSpaceSettings {maxAngularAcceleration=35, maxAngularVelocity=8, simSpaceAngularAlpha=1}
   }
   n75 = Anim Dynamics - Bone: dyn_tail_left_02_02_cr!(ComponentPose=@n74, Alpha=1, ComponentLinearAccScale:Vector="1,1,1", ComponentLinearVelScale:Vector="1,1,1", ExternalForce="0,0,0") -> (Pose) {
    angularDampingOverride=1
    angularSpringConstant=20000
    bAngularSpring=true
    bOverrideAngularDamping=true
    bUseGravityOverride=true
    gravityScale=0.35
    simulationSpace=RootRelative
    boundBone {boneName=dyn_tail_left_02_02_cr}
    componentAppliedLinearAccClamp {x=3000, z=3000}
    componentLinearAccScale=<1 1 1>
    componentLinearVelScale=<1 1 1>
    gravityOverride {z=-980}
    physicsBodyDefinitions [{
      boundBone {boneName=dyn_tail_left_02_02_cr}
      boxExtents=<40 40 40>
      constraintSetup {angularLimitsMax=<0 5 5>, angularLimitsMin=<0 -5 -15>, angularTarget=<1 0 0>}
      localJointOffset=<4 0 0>
     }]
    simSpaceSettings {maxAngularAcceleration=35, maxAngularVelocity=8, simSpaceAngularAlpha=1}
   }
   n76 = Anim Dynamics - Bone: dyn_tail_right_01_01_cr!(ComponentPose=@n75, Alpha=1, ComponentLinearAccScale:Vector="1,1,1", ComponentLinearVelScale:Vector="1,1,1", ExternalForce="0,0,0") -> (Pose) {
    angularDampingOverride=1
    angularSpringConstant=20000
    bAngularSpring=true
    bOverrideAngularDamping=true
    bUseGravityOverride=true
    gravityScale=0.35
    simulationSpace=RootRelative
    boundBone {boneName=dyn_tail_right_01_01_cr}
    componentAppliedLinearAccClamp {x=3000, z=3000}
    componentLinearAccScale=<1 1 1>
    componentLinearVelScale=<1 1 1>
    gravityOverride {z=-980}
    physicsBodyDefinitions [{
      boundBone {boneName=dyn_tail_right_01_01_cr}
      boxExtents=<20 20 20>
      constraintSetup {angularLimitsMax=<0 5 5>, angularLimitsMin=<0 -5 -20>, angularTarget=<1 0 0>}
      localJointOffset=<4 0 0>
     }]
    simSpaceSettings {maxAngularAcceleration=35, maxAngularVelocity=8, simSpaceAngularAlpha=1}
   }
   n77 = Anim Dynamics - Bone: dyn_tail_right_01_02_cr!(ComponentPose=@n76, Alpha=1, ComponentLinearAccScale:Vector="1,1,1", ComponentLinearVelScale:Vector="1,1,1", ExternalForce="0,0,0") -> (Pose) {
    angularDampingOverride=1
    angularSpringConstant=20000
    bAngularSpring=true
    bOverrideAngularDamping=true
    bUseGravityOverride=true
    gravityScale=0.35
    simulationSpace=RootRelative
    boundBone {boneName=dyn_tail_right_01_02_cr}
    componentAppliedLinearAccClamp {x=3000, z=3000}
    componentLinearAccScale=<1 1 1>
    componentLinearVelScale=<1 1 1>
    gravityOverride {z=-980}
    physicsBodyDefinitions [{
      boundBone {boneName=dyn_tail_right_01_02_cr}
      boxExtents=<40 40 40>
      constraintSetup {angularLimitsMax=<0 5 0>, angularLimitsMin=<0 -5 -15>, angularTarget=<1 0 0>}
      localJointOffset=<4 0 0>
     }]
    simSpaceSettings {maxAngularAcceleration=35, maxAngularVelocity=8, simSpaceAngularAlpha=1}
   }
   n78 = Anim Dynamics - Bone: dyn_tail_right_01_03_cr!(ComponentPose=@n77, Alpha=1, ComponentLinearAccScale:Vector="1,1,1", ComponentLinearVelScale:Vector="1,1,1", ExternalForce="0,0,0") -> (Pose) {
    angularDampingOverride=1
    angularSpringConstant=20000
    bAngularSpring=true
    bOverrideAngularDamping=true
    bUseGravityOverride=true
    gravityScale=0.35
    simulationSpace=RootRelative
    boundBone {boneName=dyn_tail_right_01_03_cr}
    componentAppliedLinearAccClamp {x=3000, z=3000}
    componentLinearAccScale=<1 1 1>
    componentLinearVelScale=<1 1 1>
    gravityOverride {z=-980}
    physicsBodyDefinitions [{
      boundBone {boneName=dyn_tail_right_01_03_cr}
      boxExtents=<40 40 40>
      constraintSetup {angularLimitsMax=<0 5 0>, angularLimitsMin=<0 -5 -15>, angularTarget=<1 0 0>}
      localJointOffset=<4 0 0>
     }]
    simSpaceSettings {maxAngularAcceleration=35, maxAngularVelocity=8, simSpaceAngularAlpha=1}
   }
   n79 = Anim Dynamics - Bone: dyn_tail_right_02_01_cr!(ComponentPose=@n78, Alpha=1, ComponentLinearAccScale:Vector="1,1,1", ComponentLinearVelScale:Vector="1,1,1", ExternalForce="0,0,0") -> (Pose) {
    angularDampingOverride=1
    angularSpringConstant=20000
    bAngularSpring=true
    bOverrideAngularDamping=true
    bUseGravityOverride=true
    gravityScale=0.35
    simulationSpace=RootRelative
    boundBone {boneName=dyn_tail_right_02_01_cr}
    componentAppliedLinearAccClamp {x=3000, z=3000}
    componentLinearAccScale=<1 1 1>
    componentLinearVelScale=<1 1 1>
    gravityOverride {z=-980}
    physicsBodyDefinitions [{
      boundBone {boneName=dyn_tail_right_02_01_cr}
      boxExtents=<40 40 40>
      constraintSetup {angularLimitsMax=<0 5 5>, angularLimitsMin=<0 -5 -15>, angularTarget=<1 0 0>}
      localJointOffset=<4 0 0>
     }]
    simSpaceSettings {maxAngularAcceleration=35, maxAngularVelocity=8, simSpaceAngularAlpha=1}
   }
   n80 = Anim Dynamics - Bone: dyn_tail_right_02_02_cr!(ComponentPose=@n79, Alpha=1, ComponentLinearAccScale:Vector="1,1,1", ComponentLinearVelScale:Vector="1,1,1", ExternalForce="0,0,0") -> (Pose) {
    angularDampingOverride=1
    angularSpringConstant=20000
    bAngularSpring=true
    bOverrideAngularDamping=true
    bUseGravityOverride=true
    gravityScale=0.35
    simulationSpace=RootRelative
    boundBone {boneName=dyn_tail_right_02_02_cr}
    componentAppliedLinearAccClamp {x=3000, z=3000}
    componentLinearAccScale=<1 1 1>
    componentLinearVelScale=<1 1 1>
    gravityOverride {z=-980}
    physicsBodyDefinitions [{
      boundBone {boneName=dyn_tail_right_02_02_cr}
      boxExtents=<40 40 40>
      constraintSetup {angularLimitsMax=<0 5 0>, angularLimitsMin=<0 -5 -15>, angularTarget=<1 0 0>}
      localJointOffset=<4 0 0>
     }]
    simSpaceSettings {maxAngularAcceleration=35, maxAngularVelocity=8, simSpaceAngularAlpha=1}
   }
   n81 = Anim Dynamics - Bone: dyn_tail_center_01_01_cr!(ComponentPose=@n80, Alpha=1, ComponentLinearAccScale:Vector="1,1,1", ComponentLinearVelScale:Vector="1,1,1", ExternalForce="0,0,0") -> (Pose) {
    angularDampingOverride=1
    angularSpringConstant=20000
    bAngularSpring=true
    bOverrideAngularDamping=true
    bUseGravityOverride=true
    gravityScale=0.35
    simulationSpace=RootRelative
    boundBone {boneName=dyn_tail_center_01_01_cr}
    componentAppliedLinearAccClamp {x=3000, z=3000}
    componentLinearAccScale=<1 1 1>
    componentLinearVelScale=<1 1 1>
    gravityOverride {z=-980}
    physicsBodyDefinitions [{
      boundBone {boneName=dyn_tail_center_01_01_cr}
      boxExtents=<20 20 20>
      constraintSetup {angularLimitsMax=<0 5 5>, angularLimitsMin=<0 -5 -20>, angularTarget=<1 0 0>}
      localJointOffset=<4 0 0>
     }]
    simSpaceSettings {maxAngularAcceleration=35, maxAngularVelocity=8, simSpaceAngularAlpha=1}
   }
   n82 = Anim Dynamics - Bone: dyn_tail_center_01_02_cr!(ComponentPose=@n81, Alpha=1, ComponentLinearAccScale:Vector="1,1,1", ComponentLinearVelScale:Vector="1,1,1", ExternalForce="0,0,0") -> (Pose) {
    angularDampingOverride=1
    angularSpringConstant=20000
    bAngularSpring=true
    bOverrideAngularDamping=true
    bUseGravityOverride=true
    gravityScale=0.35
    simulationSpace=RootRelative
    boundBone {boneName=dyn_tail_center_01_02_cr}
    componentAppliedLinearAccClamp {x=3000, z=3000}
    componentLinearAccScale=<1 1 1>
    componentLinearVelScale=<1 1 1>
    gravityOverride {z=-980}
    physicsBodyDefinitions [{
      boundBone {boneName=dyn_tail_center_01_02_cr}
      boxExtents=<40 40 40>
      constraintSetup {angularLimitsMax=<0 5 5>, angularLimitsMin=<0 -5 -15>, angularTarget=<1 0 0>}
      localJointOffset=<4 0 0>
     }]
    simSpaceSettings {maxAngularAcceleration=35, maxAngularVelocity=8, simSpaceAngularAlpha=1}
   }
   n83 = Anim Dynamics - Bone: dyn_tail_center_01_03_cr!(ComponentPose=@n82, Alpha=1, ComponentLinearAccScale:Vector="1,1,1", ComponentLinearVelScale:Vector="1,1,1", ExternalForce="0,0,0") -> (Pose) {
    angularDampingOverride=1
    angularSpringConstant=20000
    bAngularSpring=true
    bOverrideAngularDamping=true
    bUseGravityOverride=true
    gravityScale=0.35
    simulationSpace=RootRelative
    boundBone {boneName=dyn_tail_center_01_03_cr}
    componentAppliedLinearAccClamp {x=3000, z=3000}
    componentLinearAccScale=<1 1 1>
    componentLinearVelScale=<1 1 1>
    gravityOverride {z=-980}
    physicsBodyDefinitions [{
      boundBone {boneName=dyn_tail_center_01_03_cr}
      boxExtents=<40 40 40>
      constraintSetup {angularLimitsMax=<0 5 5>, angularLimitsMin=<0 -5 -15>, angularTarget=<1 0 0>}
      localJointOffset=<4 0 0>
     }]
    simSpaceSettings {maxAngularAcceleration=35, maxAngularVelocity=8, simSpaceAngularAlpha=1}
   }
   n84 = Anim Dynamics - Bone: dyn_tail_center_02_01_cr!(ComponentPose=@n83, Alpha=1, ComponentLinearAccScale:Vector="1,1,1", ComponentLinearVelScale:Vector="1,1,1", ExternalForce="0,0,0") -> (Pose) {
    angularDampingOverride=1
    angularSpringConstant=20000
    bAngularSpring=true
    bOverrideAngularDamping=true
    bUseGravityOverride=true
    gravityScale=0.35
    simulationSpace=RootRelative
    boundBone {boneName=dyn_tail_center_02_01_cr}
    componentAppliedLinearAccClamp {x=3000, z=3000}
    componentLinearAccScale=<1 1 1>
    componentLinearVelScale=<1 1 1>
    gravityOverride {z=-980}
    physicsBodyDefinitions [{
      boundBone {boneName=dyn_tail_center_02_01_cr}
      boxExtents=<40 40 40>
      constraintSetup {angularLimitsMax=<0 5 5>, angularLimitsMin=<0 -5 -15>, angularTarget=<1 0 0>}
      localJointOffset=<4 0 0>
     }]
    simSpaceSettings {maxAngularAcceleration=35, maxAngularVelocity=8, simSpaceAngularAlpha=1}
   }
   n85 = Anim Dynamics - Bone: dyn_tail_center_02_02_cr!(ComponentPose=@n84, Alpha=1, ComponentLinearAccScale:Vector="1,1,1", ComponentLinearVelScale:Vector="1,1,1", ExternalForce="0,0,0") -> (Pose) {
    angularDampingOverride=1
    angularSpringConstant=20000
    bAngularSpring=true
    bOverrideAngularDamping=true
    bUseGravityOverride=true
    gravityScale=0.35
    simulationSpace=RootRelative
    boundBone {boneName=dyn_tail_center_02_02_cr}
    componentAppliedLinearAccClamp {x=3000, z=3000}
    componentLinearAccScale=<1 1 1>
    componentLinearVelScale=<1 1 1>
    gravityOverride {z=-980}
    physicsBodyDefinitions [{
      boundBone {boneName=dyn_tail_center_02_02_cr}
      boxExtents=<40 40 40>
      constraintSetup {angularLimitsMax=<0 5 5>, angularLimitsMin=<0 -5 -15>, angularTarget=<1 0 0>}
      localJointOffset=<4 0 0>
     }]
    simSpaceSettings {maxAngularAcceleration=35, maxAngularVelocity=8, simSpaceAngularAlpha=1}
   }
   n86 = Anim Dynamics - Bone: dyn_tie_upper_01_rt!(ComponentPose=@n85, Alpha=1, ComponentLinearAccScale:Vector="1,1,1", ComponentLinearVelScale:Vector="1,1,1", ExternalForce="0,0,0") -> (Pose) {
    angularDampingOverride=1
    angularSpringConstant=20000
    bAngularSpring=true
    bOverrideAngularDamping=true
    bUseGravityOverride=true
    gravityScale=0.35
    simulationSpace=RootRelative
    boundBone {boneName=dyn_tie_upper_01_rt}
    componentAppliedLinearAccClamp=<3000 3000 3000>
    componentLinearAccScale=<1 1 1>
    componentLinearVelScale=<1 1 1>
    gravityOverride {z=-980}
    physicsBodyDefinitions [{
      boundBone {boneName=dyn_tie_upper_01_rt}
      boxExtents=<50 50 50>
      constraintSetup {angularLimitsMax=<0 10 0>, angularLimitsMin=<0 -15 0>, angularTarget=<1 0 0>}
      localJointOffset=<4 0 0>
     }]
    simSpaceSettings {maxAngularAcceleration=35, maxAngularVelocity=8, simSpaceAngularAlpha=1}
   }
   n87 = Anim Dynamics - Bone: dyn_tie_upper_02_rt!(ComponentPose=@n86, Alpha=1, ComponentLinearAccScale:Vector="1,1,1", ComponentLinearVelScale:Vector="1,1,1", ExternalForce="0,0,0") -> (Pose) {
    angularDampingOverride=1
    angularSpringConstant=20000
    bAngularSpring=true
    bOverrideAngularDamping=true
    bUseGravityOverride=true
    gravityScale=0.35
    simulationSpace=RootRelative
    boundBone {boneName=dyn_tie_upper_02_rt}
    componentAppliedLinearAccClamp=<3000 3000 3000>
    componentLinearAccScale=<1 1 1>
    componentLinearVelScale=<1 1 1>
    gravityOverride {z=-980}
    physicsBodyDefinitions [{
      boundBone {boneName=dyn_tie_upper_02_rt}
      boxExtents=<50 50 50>
      constraintSetup {angularLimitsMax=<0 15 0>, angularLimitsMin=<0 -15 0>, angularTarget=<1 0 0>}
      localJointOffset=<4 0 0>
     }]
    simSpaceSettings {maxAngularAcceleration=35, maxAngularVelocity=8, simSpaceAngularAlpha=1}
   }
   n88 = Anim Dynamics - Bone: dyn_tie_upper_01_lf!(ComponentPose=@n87, Alpha=1, ComponentLinearAccScale:Vector="1,1,1", ComponentLinearVelScale:Vector="1,1,1", ExternalForce="0,0,0") -> (Pose) {
    angularDampingOverride=1
    angularSpringConstant=20000
    bAngularSpring=true
    bOverrideAngularDamping=true
    bUseGravityOverride=true
    gravityScale=0.35
    simulationSpace=RootRelative
    boundBone {boneName=dyn_tie_upper_01_lf}
    componentAppliedLinearAccClamp=<3000 3000 3000>
    componentLinearAccScale=<1 1 1>
    componentLinearVelScale=<1 1 1>
    gravityOverride {z=-980}
    physicsBodyDefinitions [{
      boundBone {boneName=dyn_tie_upper_01_lf}
      boxExtents=<40 40 40>
      constraintSetup {angularLimitsMax=<0 5 0>, angularLimitsMin=<0 -5 0>, angularTarget=<1 0 0>}
      localJointOffset=<4 0 0>
     }]
    simSpaceSettings {maxAngularAcceleration=35, maxAngularVelocity=8, simSpaceAngularAlpha=1}
   }
   n89 = Anim Dynamics - Bone: dyn_tie_upper_02_lf!(ComponentPose=@n88, Alpha=1, ComponentLinearAccScale:Vector="1,1,1", ComponentLinearVelScale:Vector="1,1,1", ExternalForce="0,0,0") -> (Pose) {
    angularDampingOverride=1
    angularSpringConstant=20000
    bAngularSpring=true
    bOverrideAngularDamping=true
    bUseGravityOverride=true
    gravityScale=0.35
    simulationSpace=RootRelative
    boundBone {boneName=dyn_tie_upper_02_lf}
    componentAppliedLinearAccClamp=<3000 3000 3000>
    componentLinearAccScale=<1 1 1>
    componentLinearVelScale=<1 1 1>
    gravityOverride {z=-980}
    physicsBodyDefinitions [{
      boundBone {boneName=dyn_tie_upper_02_lf}
      boxExtents=<40 40 40>
      constraintSetup {angularLimitsMax=<0 15 0>, angularLimitsMin=<0 -15 0>, angularTarget=<1 0 0>}
      localJointOffset=<4 0 0>
     }]
    simSpaceSettings {maxAngularAcceleration=35, maxAngularVelocity=8, simSpaceAngularAlpha=1}
   }
   n90 = Anim Dynamics - Bone: dyn_tie_lower_01_lf!(ComponentPose=@n89, Alpha=1, ComponentLinearAccScale:Vector="1,1,1", ComponentLinearVelScale:Vector="1,1,1", ExternalForce="0,0,0") -> (Pose) {
    angularDampingOverride=1
    angularSpringConstant=20000
    bAngularSpring=true
    bOverrideAngularDamping=true
    bUseGravityOverride=true
    gravityScale=0.35
    simulationSpace=RootRelative
    boundBone {boneName=dyn_tie_lower_01_lf}
    componentAppliedLinearAccClamp=<3000 3000 3000>
    componentLinearAccScale=<1 1 1>
    componentLinearVelScale=<1 1 1>
    gravityOverride {z=-980}
    physicsBodyDefinitions [{
      boundBone {boneName=dyn_tie_lower_01_lf}
      boxExtents=<30 30 30>
      constraintSetup {angularLimitsMax=<0 15 0>, angularLimitsMin=<0 -15 0>, angularTarget=<1 0 0>}
      localJointOffset=<4 0 0>
     }]
    simSpaceSettings {maxAngularAcceleration=35, maxAngularVelocity=8, simSpaceAngularAlpha=1}
   }
   n91 = Anim Dynamics - Bone: dyn_tie_lower_02_lf!(ComponentPose=@n90, Alpha=1, ComponentLinearAccScale:Vector="1,1,1", ComponentLinearVelScale:Vector="1,1,1", ExternalForce="0,0,0") -> (Pose) {
    angularDampingOverride=1
    angularSpringConstant=20000
    bAngularSpring=true
    bOverrideAngularDamping=true
    bUseGravityOverride=true
    gravityScale=0.35
    simulationSpace=RootRelative
    boundBone {boneName=dyn_tie_lower_02_lf}
    componentAppliedLinearAccClamp=<3000 3000 3000>
    componentLinearAccScale=<1 1 1>
    componentLinearVelScale=<1 1 1>
    gravityOverride {z=-980}
    physicsBodyDefinitions [{
      boundBone {boneName=dyn_tie_lower_02_lf}
      boxExtents=<30 30 30>
      constraintSetup {angularLimitsMax=<0 15 0>, angularLimitsMin=<0 -15 0>, angularTarget=<1 0 0>}
      localJointOffset=<4 0 0>
     }]
    simSpaceSettings {maxAngularAcceleration=35, maxAngularVelocity=8, simSpaceAngularAlpha=1}
   }
   n92 = Anim Dynamics - Bone: dyn_tie_lower_01_rt!(ComponentPose=@n91, Alpha=1, ComponentLinearAccScale:Vector="1,1,1", ComponentLinearVelScale:Vector="1,1,1", ExternalForce="0,0,0") -> (Pose) {
    angularDampingOverride=1
    angularSpringConstant=20000
    bAngularSpring=true
    bOverrideAngularDamping=true
    bUseGravityOverride=true
    gravityScale=0.35
    simulationSpace=RootRelative
    boundBone {boneName=dyn_tie_lower_01_rt}
    componentAppliedLinearAccClamp=<3000 3000 3000>
    componentLinearAccScale=<1 1 1>
    componentLinearVelScale=<1 1 1>
    gravityOverride {z=-980}
    physicsBodyDefinitions [{
      boundBone {boneName=dyn_tie_lower_01_rt}
      boxExtents=<30 30 30>
      constraintSetup {angularLimitsMax=<0 15 0>, angularLimitsMin=<0 -15 0>, angularTarget=<1 0 0>}
      localJointOffset=<6 0 0>
     }]
    simSpaceSettings {maxAngularAcceleration=35, maxAngularVelocity=8, simSpaceAngularAlpha=1}
   }
   n93 = Anim Dynamics - Bone: dyn_tie_lower_02_rt!(ComponentPose=@n92, Alpha=1, ComponentLinearAccScale:Vector="1,1,1", ComponentLinearVelScale:Vector="1,1,1", ExternalForce="0,0,0") -> (Pose) {
    angularDampingOverride=1
    angularSpringConstant=20000
    bAngularSpring=true
    bOverrideAngularDamping=true
    bUseGravityOverride=true
    gravityScale=0.35
    simulationSpace=RootRelative
    boundBone {boneName=dyn_tie_lower_02_rt}
    componentAppliedLinearAccClamp=<3000 3000 3000>
    componentLinearAccScale=<1 1 1>
    componentLinearVelScale=<1 1 1>
    gravityOverride {z=-980}
    physicsBodyDefinitions [{
      boundBone {boneName=dyn_tie_lower_02_rt}
      boxExtents=<30 30 30>
      constraintSetup {angularLimitsMax=<0 15 0>, angularLimitsMin=<0 -15 0>, angularTarget=<1 0 0>}
      localJointOffset=<4 0 0>
     }]
    simSpaceSettings {maxAngularAcceleration=35, maxAngularVelocity=8, simSpaceAngularAlpha=1}
   }
   n94 = RigidBody!(ComponentPose=@n93, Alpha=0.8, OverrideWorldGravity:Vector="0,0,-980", ExternalForce) -> (Pose) {
    bClampLinearTranslationLimitToRefPose=true
    bOverrideWorldGravity=true
    overridePhysicsAsset=&/F_MED_RoosterRoast_Physics
    simulationSpace=BaseBoneSpace
    baseBoneRef {boneName=pelvis}
    overrideWorldGravity {z=-980}
    simSpaceSettings {worldAlpha=0.2}
   }
   Save cached pose 'AnimDynPose'(Pose=Component To Local(ComponentPose=@n94)) {cacheName=AnimDynPose}
  ]}
 EventGraph {nodes [
   Event OnCharacterCustomizationLoaded()
   > n1 = Cast To PlayerPawn_Athena(Object=TryGetPawnOwner())
   > Set Body_Mesh(Body_Mesh=GetSkeletalMeshForPartType(self=@n1))
   n2 = CheckForBackbling()
   > Set isWearingBackbling(isWearingBackbling=AND Boolean!(A=@n2, B=NOT Boolean(A=@n2.bIsBackblingACape), C:bool=NOT Boolean(A=@n2.bIsBackblingLong)) -> (ReturnValue))
   > Set Is Wearing Cape(Is Wearing Cape=OR Boolean(A=@n2.bIsBackblingACape, B=@n2.bIsBackblingLong))
   Event BlueprintUpdateAnimation()
   > n3 = Sequence!(execute) -> (then_0, then_1, then_2:Exec, then_3:Exec)
   > n4 = "Gravity Override State Function (Parameter Asset) (Nativized)"(
    CharacterPartAnimInstance=Self-Reference()
    GameTimeInSeconds=GetGameTimeInSeconds()
    Parameters=&/F_MED_RoosterRoast_Parameters
   )
   > Set GravityOverride(GravityOverride=@n4)
   n5 = Get bIsSprinting()
   n6 = Get bIsPlayingEmote()
   > Set sprint(sprint:double="To Float (Boolean)"(InBool=OR Boolean(A=@n5, B=@n6))) -> (Output_Get:double)
   > Set crouch(crouch:double="To Float (Boolean)"(InBool=Get bIsCrouching())) -> (Output_Get:double)
   > Set emote(emote:double="To Float (Boolean)"(InBool=Get bIsPlayingEmote())) -> (Output_Get:double)
   > Set slide off(slide off:double=SelectFloat(A=0, B=1, bPickA=Get bIsSliding())) -> (Output_Get:double)
   > n7 = Cast To FortPlayerAnimInstance(Object=GetAnimInstance(self=GetOwnerSkeletalMeshComponent()))
   > Set pistol(pistol:double=-0.6) -> (Output_Get:double)
   > n8 = Switch on EFortWeaponCoreAnimation!(execute, Selection:EFortWeaponCoreAnimation Enum=Get WeaponCoreAnim(self=@n7)) -> (
    Melee:Exec
    Pistol:Exec
    Shotgun:Exec
    PaperBlueprint:Exec
    Rifle:Exec
    MeleeOneHand:Exec
    MachinePistol:Exec
    RocketLauncher:Exec
    GrenadeLauncher:Exec
    GoingCommando:Exec
    AssaultRifle:Exec
    TacticalShotgun:Exec
    SniperRifle:Exec
    TrapPlacement:Exec
    ShoulderLauncher:Exec
    AbilityDecoTool:Exec
    Crossbow:Exec
    C4:Exec
    RemoteControl:Exec
    DualWield:Exec
    AR_BullPup:Exec
    AR_ForwardGrip:Exec
    MedPackPaddles:Exec
    SMG_P90:Exec
    AR_DrumGun:Exec
    Consumable_Small:Exec
    Consumable_Large:Exec
    Balloon:Exec
    MountedTurret:Exec
    CreativeTool:Exec
    ExplosiveBow:Exec
    AshtonIndigo:Exec
    AshtonChicago:Exec
    MeleeDualWield:Exec
    Unarmed:Exec
   )
   Set pistol(execute=@n8.Pistol, pistol=-0.1)
   n9 = Get bIsPlayingEmote()
   n10 = Get bIsSkyDiving()
   n11 = Branch(execute=@n3.then_1, Condition=AND Boolean(A=Get bIsInFrontEnd(), B=NOT Boolean(A=@n9), C=NOT Boolean(A=@n10)))
   > Set frontend(frontend:double=1) -> (Output_Get:double)
   Set frontend(execute=@n11.else, frontend=0)
   n12 = Get bIsMoving2D()
   Set CrouchIDLE(execute=@n3.then_2, CrouchIDLE:bool=AND Boolean!(A=Get bIsCrouching(), B=NOT Boolean(A=@n12)) -> (ReturnValue)) -> (Output_Get:bool)
   n13 = Get bIsCrouchSprinting()
   > Set is Use Pickaxe Gauntlet(is Use Pickaxe Gauntlet:bool=AND Boolean!(
     A=Get bIsLeftArmHidden() -> (bIsLeftArmHidden:bool)
     B=Get bIsRightArmHidden() -> (bIsRightArmHidden:bool)
     C:bool=Get bIsCrouching()
     D:bool=Get bIsPlayingMeleeAnim()
     E:bool=NOT Boolean(A=@n13)
    ) -> (ReturnValue)) -> (Output_Get:bool)
   > Set IsCrouchMelee(IsCrouchMelee:bool=AND Boolean!(A=Get bIsCrouching(), B=Get bIsPlayingMeleeAnim()) -> (ReturnValue)) -> (Output_Get:bool)
   > Set IsWearingCapeFloating(IsWearingCapeFloating:bool=AND Boolean(A=Get isWearingBackbling(), B=Get isWearingBackbling())) -> (Output_Get:bool)
   n14 = Get bIsCrouching()
   n15 = Get bIsSprinting()
   n16 = Get bIsTacticalSprinting()
   > n17 = Branch(Condition=AND Boolean(A=NOT Boolean(A=@n14), B=OR Boolean(A=@n15, B=@n16)))
   > Set Running(Running:double=0.5) -> (Output_Get:double)
   Set Running(execute=@n17.else, Running=1)
   n18 = Event BlueprintProcessAnimInputEvent()
   n19 = IsNamed(self=@n18.AnimInputEvent, InEventName=aie.Player.Zipline.Entered)
   n20 = OR Boolean(A=Get bIsMoving2D(), B=Get bIsTacticalSprinting())
   n21 = Get bIsJumping()
   > n22 = Branch(Condition=OR Boolean(A=@n19, B=AND Boolean!(A=@n19, B=@n21, C:bool=@n20) -> (ReturnValue)))
   > Set animDyn on zipline(animDyn on zipline:Vector="0, 0, 0") -> (Output_Get:Vector)
   Branch(execute=@n22.else, Condition=IsNamed(self=@n18.AnimInputEvent, InEventName=aie.Player.Zipline.Exited))
   > Set animDyn on zipline(animDyn on zipline="1,1,1")
   Get IsWearingCapeFloating()
  ]}
]
variables [
 animDyn on zipline {type=Vector, default=<1 1 1>}
 Body_Mesh {type=SkeletalMeshComponent}
 crouch {type=number}
 CrouchIDLE {type=boolean}
 emote {type=number}
 frontend {type=number}
 Gravity Override Params {type=GravityOverrideParamsStruct}
 GravityOverride {type=Vector}
 is Use Pickaxe Gauntlet {type=boolean}
 Is Wearing Cape {type=boolean}
 IsCrouchMelee {type=boolean}
 isWearingBackbling {type=boolean}
 IsWearingCapeFloating {type=boolean}
 pistol {type=number}
 Running {type=number, default=1}
 slide off {type=number, default=1}
 sprint {type=number}
]
