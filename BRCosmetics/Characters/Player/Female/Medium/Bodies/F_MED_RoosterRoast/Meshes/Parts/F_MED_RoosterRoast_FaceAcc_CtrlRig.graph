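# Control Rig for the F_MED_RoosterRoast face accessory mesh that positions and orients the necklace's base bone from the upper body.
# The forward solve parent-constrains the necklace base to spine_05, spine_04 and clavicle_r, and an aim helper bone to the two spine bones, both weighted mostly to spine_05.
# An offset-preserving aim constraint then points the base at that helper, using it as up-vector space.
# Hat, crest and necklace dyn bones are not driven directly.
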ControlRigBlueprint /BRCosmetics/Characters/Player/Female/Medium/Bodies/F_MED_RoosterRoast/Meshes/Parts/F_MED_RoosterRoast_FaceAcc_CtrlRig
graphs [
 RigVMFunctionLibrary {links [], nodes []}
 RigVMModel {nodes [
   RigUnit_BeginExecution = &RigUnit_BeginExecution()
   &RigUnit_ParentConstraint(
    AdvancedSettings=(RotationOrderForFilter=XZY)
    Child=(Type=Bone,Name="base_necklace_01_cr")
    Parents=((Item=(Type=Bone,Name="clavicle_r"),Weight=0.0359),(Item=(Type=Bone,Name="spine_04"),Weight=0.0497),(Item=(Type=Bone,Name="spine_05"),Weight=0.9143))
    Weight=1
    ExecutePin=@RigUnit_BeginExecution.ExecutePin
   )
   > &RigUnit_ParentConstraint(
    AdvancedSettings=(RotationOrderForFilter=XZY)
    Child=(Type=Bone,Name="aim_necklace_cr")
    Parents=((Item=(Type=Bone,Name="spine_04"),Weight=0.1915),(Item=(Type=Bone,Name="spine_05"),Weight=0.8085))
    Weight=1
   )
   > &RigUnit_AimConstraintLocalSpaceOffset(
    Child=(Type=Bone,Name="base_necklace_01_cr")
    Filter=(bX=True,bY=True,bZ=True)
    Parents=((Item=(Type=Bone,Name="aim_necklace_cr"),Weight=1))
    Weight=1
    WorldUp=(Target=(X=0,Y=0,Z=1),Kind=Direction,Space=(Type=Bone,Name="aim_necklace_cr"))
   )
  ]}
]
hierarchy [
 root {loc=<0 0 0>, rot=<0 0 0 1>}
 pelvis {parent=root, loc=<0 -0.5301 98.0571>, rot=<0.0127 0.707 -0.0127 -0.707>}
 spine_01 {parent=pelvis, loc=<6.1208 -0.9328 0>, rot=<0 0 0.034 -0.9994>}
 spine_02 {parent=spine_01, loc=<6.5996 -0.2095 0>, rot=<0 0 0.0238 -0.9997>}
 spine_03 {parent=spine_02, loc=<6.4914 -0.1237 0>, rot=<0 0 0.0149 -0.9999>}
 spine_04 {parent=spine_03, loc=<6.9598 0.0794 0>, rot=<0 0 -0.0132 -0.9999>}
 spine_05 {parent=spine_04, loc=<6.9987 0 0>, rot=<0 0 -0.0015 -1>}
 neck_01 {parent=spine_05, loc=<11.3138 0.1529 0>, rot=<0 0 -0.2003 -0.9797>}
 neck_02 {parent=neck_01, loc=<4.5627 0 0>, rot=<0 0 0.0504 -0.9987>}
 head {parent=neck_02, loc=<5.3914 0 0>, rot=<0 0 0.115 -0.9934>}
 base_hat_cr {parent=head, loc=<16.6655 -9.9148 0.126>, rot=<0.5107 -0.4891 0.5066 -0.4933>}
 dyn_hat_01_cr {parent=base_hat_cr, loc=<0 0 0>, rot=<0 0 0 1>}
 dyn_hat_02_cr {parent=dyn_hat_01_cr, loc=<6.0035 0 0>, rot=<0 0 0 1>}
 base_crest_01_cr {parent=head, loc=<19.9323 3.5221 0.3459>, rot=<-0.0241 -0.1463 -0.2811 -0.9482>}
 dyn_crest_01_cr {parent=base_crest_01_cr, loc=<0 0 0>, rot=<0 0 0 1>}
 dyn_crest_02_cr {parent=dyn_crest_01_cr, loc=<6.3281 0 1.4456>, rot=<0 -0.1425 0 -0.9898>}
 dyn_crest_03_cr {parent=dyn_crest_02_cr, loc=<6.1564 0 0>, rot=<0 -0.5053 0 -0.863>}
 dyn_crest_upper_01_cr {parent=dyn_crest_03_cr, loc=<0.4516 -2.9153 -0.0647>, rot=<0.0301 0.4577 0.0375 -0.8878>}
 dyn_crest_middle_01_cr {parent=dyn_crest_03_cr, loc=<5.3069 2.2724 -3.8005>, rot=<-0.0841 -0.3222 -0.1417 -0.9322>}
 dyn_crest_lower_02_cr {parent=dyn_crest_03_cr, loc=<3.3992 6.0926 -1.9705>, rot=<-0.032 -0.3038 -0.2079 -0.9292>}
 dyn_crest_lower_03_cr {parent=dyn_crest_lower_02_cr, loc=<6.82 0 0>, rot=<-0.0288 0 0 -0.9996>}
 base_necklace_01_cr {parent=neck_01, loc=<1.3257 1.0749 0>, rot=<-0.5213 -0.4777 -0.4777 -0.5213>}
 dyn_necklace_01_cr {parent=base_necklace_01_cr, loc=<0 0 0>, rot=<0 0 0 1>}
 dyn_necklace_02_cr {parent=dyn_necklace_01_cr, loc=<5.6025 0 0>, rot=<0 -0.0875 0 -0.9962>}
 dyn_necklace_03_cr {parent=dyn_necklace_02_cr, loc=<5.6025 0 0>, rot=<0 -0.1217 0 -0.9926>}
 dyn_necklace_04_cr {parent=dyn_necklace_03_cr, loc=<4.9977 0 0>, rot=<0 -0.2954 0 -0.9554>}
 clavicle_l {parent=spine_05, loc=<5.4831 -0.1053 -3.2946>, rot=<0.0465 -0.7289 -0.009 -0.6829>}
 clavicle_r {parent=spine_05, loc=<5.4831 -0.1053 3.2946>, rot=<0.7289 0.0465 -0.6829 0.009>}
 dfrm_clavicle_l {parent=spine_05, loc=<5.4831 -0.1053 -3.2946>, rot=<0.0465 -0.7289 -0.009 -0.6829>}
 dfrm_upperarm_l {parent=dfrm_clavicle_l, loc=<9.8658 -1.2136 -0.4105>, rot=<0.0164 -0.3245 -0.0269 -0.9454>}
 dfrm_upperarm_twist_01_l {parent=dfrm_upperarm_l, loc=<7.2135 -0.1134 0>, rot=<0 0 0 1>}
 dfrm_clavicle_r {parent=spine_05, loc=<5.4831 -0.1053 3.2946>, rot=<0.7289 0.0465 -0.6829 0.009>}
 dfrm_upperarm_r {parent=dfrm_clavicle_r, loc=<-9.8658 1.2136 0.4105>, rot=<0.0164 -0.3245 -0.0269 -0.9454>}
 dfrm_upperarm_twist_01_r {parent=dfrm_upperarm_r, loc=<-7.2135 0.1134 0>, rot=<0 0 0 -1>}
 aim_necklace_cr {parent=spine_05, loc=<6.5082 18.7949 0>, rot=<0.4151 0.5725 0.5725 0.4151>}
]
