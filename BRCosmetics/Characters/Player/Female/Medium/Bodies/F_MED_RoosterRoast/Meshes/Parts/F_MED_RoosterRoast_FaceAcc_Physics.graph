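# Physics asset for the RoosterRoast female medium character's face accessory part, setting up physics for a dangling necklace on the upper torso.
# Three capsule bodies on the dyn_necklace bones form a chain attached to a large spine_05 capsule that stands in for the torso.
# The first link rotates freely against the spine, while the two lower links lock twist and allow only narrow swing, keeping the pendant fairly stiff.
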PhysicsAsset /BRCosmetics/Characters/Player/Female/Medium/Bodies/F_MED_RoosterRoast/Meshes/Parts/F_MED_RoosterRoast_FaceAcc_Physics
bodies {
 dyn_necklace_02_cr [{shapeType=Capsule, length=2.4387, radius=2.5613, shapeName=dyn_necklace_02_cr_capsule, rotation=<rot 90 0 0>}]
 dyn_necklace_03_cr [{shapeType=Capsule, length=2.9755, radius=2.0245, shapeName=dyn_necklace_03_cr_capsule, center=<1.8008 0 0>, rotation=<rot 90 0 0>}]
 dyn_necklace_04_cr [{
   shapeType=Capsule
   length=6.1628
   radius=1.8453
   shapeName=dyn_necklace_04_cr_capsule
   center=<4.1239 -0.059 0.3106>
   rotation=<rot 0 90.174 -88.3166>
  }]
 spine_05 [{shapeType=Capsule, length=7.5755, radius=10.6416, center=<-3.0163 4.9222 -0.0208>, rotation=<rot 0 64.3566 0.2455>}]
}
constraints [
 {
  bone1Name=dyn_necklace_02_cr
  bone2Name=spine_05
  swing1LimitDegrees=45
  swing1Motion=Free
  swing2LimitDegrees=45
  swing2Motion=Free
  twistLimitDegrees=45
  twistMotion=Free
 }
 {
  bone1Name=dyn_necklace_03_cr
  bone2Name=dyn_necklace_02_cr
  swing1LimitDegrees=10
  swing2LimitDegrees=10
  twistLimitDegrees=45
  twistMotion=Locked
 }
 {
  bone1Name=dyn_necklace_04_cr
  bone2Name=dyn_necklace_03_cr
  swing1LimitDegrees=20
  swing2LimitDegrees=20
  twistLimitDegrees=45
  twistMotion=Locked
 }
]
